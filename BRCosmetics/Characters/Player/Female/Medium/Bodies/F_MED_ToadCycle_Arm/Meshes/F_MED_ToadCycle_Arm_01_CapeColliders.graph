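# Physics asset of cape colliders for the female medium ToadCycle_Arm player cosmetic body.
# Only head, neck_01, pelvis, spine_01 and spine_02 carry capsules, with the neck's being very small.
# Limb, spine_05 and the left and right four-bone dyn_cape chains have bodies without shapes.
# Constraints link the bodies in a tree, both cape chains attached to pelvis, with all swing and twist motion free.
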PhysicsAsset /BRCosmetics/Characters/Player/Female/Medium/Bodies/F_MED_ToadCycle_Arm/Meshes/F_MED_ToadCycle_Arm_01_CapeColliders
bodies {
 calf_l []
 calf_r []
 dyn_cape_b_l_01 []
 dyn_cape_b_l_02 []
 dyn_cape_b_l_03 []
 dyn_cape_b_l_04 []
 dyn_cape_b_r_01 []
 dyn_cape_b_r_02 []
 dyn_cape_b_r_03 []
 dyn_cape_b_r_04 []
 foot_l []
 foot_r []
 hand_l []
 hand_r []
 head [{shapeType=Capsule, length=0.505, radius=20.505, shapeName=head_capsule, center=<17.0277 0 0>, rotation=<rot 0 -90 0>}]
 lowerarm_l []
 lowerarm_r []
 neck_01 [{shapeType=Capsule, length=0.505, radius=0.505, shapeName=neck_01_capsule, rotation=<rot 0 -90 0>}]
 pelvis [{shapeType=Capsule, length=8.7365, radius=13.25, center=<-3.43 -1.1514 0>, rotation=<rot 0 -0.0008 -0.0007>}]
 spine_01 [{shapeType=Capsule, length=7.9865, radius=13, center=<-3.43 -1.1514 0>, rotation=<rot 0 -0.0008 -0.0007>}]
 spine_02 [{shapeType=Capsule, length=4.9865, radius=19.25, center=<-3.43 -1.1514 0>, rotation=<rot 0 -0.0008 -0.0007>}]
 spine_05 []
 upperarm_l []
 upperarm_r []
}
constraints [
 {
  bone1Name=foot_r
  bone2Name=calf_r
  swing1LimitDegrees=45
  swing1Motion=Free
  swing2LimitDegrees=45
  swing2Motion=Free
  twistLimitDegrees=45
  twistMotion=Free
 }
 {
  bone1Name=spine_01
  bone2Name=pelvis
  swing1LimitDegrees=45
  swing1Motion=Free
  swing2LimitDegrees=45
  swing2Motion=Free
  twistLimitDegrees=45
  twistMotion=Free
 }
 {
  bone1Name=spine_02
  bone2Name=spine_01
  swing1LimitDegrees=45
  swing1Motion=Free
  swing2LimitDegrees=45
  swing2Motion=Free
  twistLimitDegrees=45
  twistMotion=Free
 }
 {
  bone1Name=calf_l
  bone2Name=pelvis
  swing1LimitDegrees=45
  swing1Motion=Free
  swing2LimitDegrees=45
  swing2Motion=Free
  twistLimitDegrees=45
  twistMotion=Free
 }
 {
  bone1Name=foot_l
  bone2Name=calf_l
  swing1LimitDegrees=45
  swing1Motion=Free
  swing2LimitDegrees=45
  swing2Motion=Free
  twistLimitDegrees=45
  twistMotion=Free
 }
 {
  bone1Name=calf_r
  bone2Name=pelvis
  swing1LimitDegrees=45
  swing1Motion=Free
  swing2LimitDegrees=45
  swing2Motion=Free
  twistLimitDegrees=45
  twistMotion=Free
 }
 {
  bone1Name=spine_05
  bone2Name=spine_02
  swing1LimitDegrees=45
  swing1Motion=Free
  swing2LimitDegrees=45
  swing2Motion=Free
  twistLimitDegrees=45
  twistMotion=Free
 }
 {
  bone1Name=upperarm_l
  bone2Name=spine_05
  swing1LimitDegrees=45
  swing1Motion=Free
  swing2LimitDegrees=45
  swing2Motion=Free
  twistLimitDegrees=45
  twistMotion=Free
 }
 {
  bone1Name=lowerarm_l
  bone2Name=upperarm_l
  swing1LimitDegrees=45
  swing1Motion=Free
  swing2LimitDegrees=45
  swing2Motion=Free
  twistLimitDegrees=45
  twistMotion=Free
 }
 {
  bone1Name=hand_l
  bone2Name=lowerarm_l
  swing1LimitDegrees=45
  swing1Motion=Free
  swing2LimitDegrees=45
  swing2Motion=Free
  twistLimitDegrees=45
  twistMotion=Free
 }
 {
  bone1Name=upperarm_r
  bone2Name=spine_05
  swing1LimitDegrees=45
  swing1Motion=Free
  swing2LimitDegrees=45
  swing2Motion=Free
  twistLimitDegrees=45
  twistMotion=Free
 }
 {
  bone1Name=lowerarm_r
  bone2Name=upperarm_r
  swing1LimitDegrees=45
  swing1Motion=Free
  swing2LimitDegrees=45
  swing2Motion=Free
  twistLimitDegrees=45
  twistMotion=Free
 }
 {
  bone1Name=hand_r
  bone2Name=lowerarm_r
  swing1LimitDegrees=45
  swing1Motion=Free
  swing2LimitDegrees=45
  swing2Motion=Free
  twistLimitDegrees=45
  twistMotion=Free
 }
 {
  bone1Name=dyn_cape_b_r_03
  bone2Name=dyn_cape_b_r_02
  swing1LimitDegrees=45
  swing1Motion=Free
  swing2LimitDegrees=45
  swing2Motion=Free
  twistLimitDegrees=45
  twistMotion=Free
 }
 {
  bone1Name=dyn_cape_b_r_04
  bone2Name=dyn_cape_b_r_03
  swing1LimitDegrees=45
  swing1Motion=Free
  swing2LimitDegrees=45
  swing2Motion=Free
  twistLimitDegrees=45
  twistMotion=Free
 }
 {
  bone1Name=dyn_cape_b_l_04
  bone2Name=dyn_cape_b_l_03
  swing1LimitDegrees=45
  swing1Motion=Free
  swing2LimitDegrees=45
  swing2Motion=Free
  twistLimitDegrees=45
  twistMotion=Free
 }
 {
  bone1Name=dyn_cape_b_l_03
  bone2Name=dyn_cape_b_l_02
  swing1LimitDegrees=45
  swing1Motion=Free
  swing2LimitDegrees=45
  swing2Motion=Free
  twistLimitDegrees=45
  twistMotion=Free
 }
 {
  bone1Name=dyn_cape_b_l_02
  bone2Name=dyn_cape_b_l_01
  swing1LimitDegrees=45
  swing1Motion=Free
  swing2LimitDegrees=45
  swing2Motion=Free
  twistLimitDegrees=45
  twistMotion=Free
 }
 {
  bone1Name=dyn_cape_b_l_01
  bone2Name=pelvis
  swing1LimitDegrees=45
  swing1Motion=Free
  swing2LimitDegrees=45
  swing2Motion=Free
  twistLimitDegrees=45
  twistMotion=Free
 }
 {
  bone1Name=dyn_cape_b_r_02
  bone2Name=dyn_cape_b_r_01
  swing1LimitDegrees=45
  swing1Motion=Free
  swing2LimitDegrees=45
  swing2Motion=Free
  twistLimitDegrees=45
  twistMotion=Free
 }
 {
  bone1Name=dyn_cape_b_r_01
  bone2Name=pelvis
  swing1LimitDegrees=45
  swing1Motion=Free
  swing2LimitDegrees=45
  swing2Motion=Free
  twistLimitDegrees=45
  twistMotion=Free
 }
 {
  bone1Name=neck_01
  bone2Name=spine_05
  swing1LimitDegrees=45
  swing1Motion=Free
  swing2LimitDegrees=45
  swing2Motion=Free
  twistLimitDegrees=45
  twistMotion=Free
 }
 {
  bone1Name=head
  bone2Name=neck_01
  swing1LimitDegrees=45
  swing1Motion=Free
  swing2LimitDegrees=45
  swing2Motion=Free
  twistLimitDegrees=45
  twistMotion=Free
 }
]
