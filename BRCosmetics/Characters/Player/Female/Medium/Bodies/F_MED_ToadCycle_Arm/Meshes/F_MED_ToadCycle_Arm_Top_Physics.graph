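# Physics asset for the top mesh of the female medium ToadCycle Arm character cosmetic, covering its scarf and cloak secondary motion.
# Capsules on the spine, right clavicle, upper arm and cloth-volume bones form body volumes alongside the dynamic bones.
# Back and front scarf chains hang from tiny base bodies on spine_05, with limited swing and very tight twist.
# Cloak, clavicle and scarf-base constraints leave rotation free.
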PhysicsAsset /BRCosmetics/Characters/Player/Female/Medium/Bodies/F_MED_ToadCycle_Arm/Meshes/F_MED_ToadCycle_Arm_Top_Physics
bodies {
 base_scarf_bk [{shapeType=Capsule, length=0.505, radius=0.505, shapeName=base_scarf_bk_capsule, rotation=<rot 0 -90 0>}]
 base_scarf_fr [{shapeType=Capsule, length=0.505, radius=0.505, shapeName=base_scarf_fr_capsule, rotation=<rot 0 -90 0>}]
 clavicle_r [{shapeType=Capsule, length=9.5, radius=9, shapeName=upperarm_r_capsule, center=<-5 0 3>, rotation=<rot -90 0 0>}]
 dyn_Cloth_volume_fn_01 [{
   shapeType=Capsule
   length=31.7342
   radius=11.5625
   shapeName=dyn_Cloth_volume_r_A01_capsule
   center=<-20.7316 9.4002 -2.3128>
   rotation=<rot 0.0252 84.6276 -89.7443>
  }]
 dyn_Cloth_volume_l_A01 [{
   shapeType=Capsule
   length=28.9459
   radius=8.0918
   shapeName=dyn_Cloth_volume_l_A01_capsule
   center=<-17.728 5.7795 2.1248>
   rotation=<rot 0 -92.5114 -81.2817>
  }]
 dyn_Cloth_volume_r_A01 [{
   shapeType=Capsule
   length=31.7342
   radius=11.5625
   shapeName=dyn_Cloth_volume_r_A01_capsule
   center=<-21.0241 9.7162 1.6888>
   rotation=<rot 0 90.2523 -86.9306>
  }]
 dyn_cloak_01 [{shapeType=Capsule, length=16.25, radius=5.5, shapeName=dyn_cloak_01_capsule, center=<7.966 -2.3815 -4.1971>, rotation=<rot -22.5001 0 90>}]
 dyn_cloak_02 [{
   shapeType=Capsule
   length=7.5
   radius=6.75
   shapeName=dyn_cloak_02_capsule
   center=<4.9973 -1.7249 -3.2189>
   rotation=<rot -0.0004 0 70.3125>
  }]
 dyn_scarf_bk_01 [{shapeType=Capsule, length=5, radius=4, center=<5 0 0.5>, rotation=<rot -90 90 0>}]
 dyn_scarf_bk_02 [{shapeType=Capsule, length=5.5, radius=4, center=<5 0 1.1242>, rotation=<rot -90 90 0>}]
 dyn_scarf_bk_03 [{shapeType=Capsule, length=6, radius=3.5, center=<5 0 2.1242>, rotation=<rot -90 90 0>}]
 dyn_scarf_bk_04 [{shapeType=Capsule, length=5.75, radius=3.25, center=<5 0 2>, rotation=<rot -90 90 0>}]
 dyn_scarf_bk_05 [{shapeType=Capsule, length=6, radius=3, center=<5 0 1.5>, rotation=<rot -90 90 0>}]
 dyn_scarf_fr_01 [{shapeType=Capsule, length=5, radius=4.5, center=<6.1903 0.4055 5.3673>, rotation=<rot -75.9377 -180 180>}]
 dyn_scarf_fr_02 [{shapeType=Capsule, length=3.5, radius=5, center=<6.1069 0.4957 3.0902>, rotation=<rot -70.3127 -180 -180>}]
 dyn_scarf_fr_03 [{shapeType=Capsule, length=6.75, radius=4.75, center=<8.0103 0.6463 3.3385>, rotation=<rot -90 0 0>}]
 spine_02 [{shapeType=Capsule, length=11.5, radius=13.75, shapeName=spine_02_capsule, center=<-1.4084 -1.7704 0.3596>}]
 spine_03 [{shapeType=Capsule, length=11, radius=13.5, shapeName=spine_03_capsule, center=<0 -0.8976 0>}]
 spine_04 [{shapeType=Capsule, length=10.25, radius=13.8882, shapeName=spine_04_capsule, center=<-0.9375 -0.5632 0>}]
 spine_05 [{shapeType=Capsule, length=12.75, radius=13.25, shapeName=spine_05_capsule, center=<-2.5484 -0.669 0>}]
 upperarm_r [{shapeType=Capsule, length=15.5, radius=8, shapeName=upperarm_r_capsule, center=<-14.6143 -0.5943 3.5872>, rotation=<rot -90 0 0>}]
}
constraints [
 {
  bone1Name=base_scarf_bk
  bone2Name=spine_05
  swing1LimitDegrees=45
  swing1Motion=Free
  swing2LimitDegrees=45
  swing2Motion=Free
  twistLimitDegrees=45
  twistMotion=Free
 }
 {bone1Name=dyn_scarf_bk_01, bone2Name=base_scarf_bk, swing1LimitDegrees=15, swing2LimitDegrees=25, twistLimitDegrees=2}
 {bone1Name=dyn_scarf_bk_02, bone2Name=dyn_scarf_bk_01, swing1LimitDegrees=15, swing2LimitDegrees=20, twistLimitDegrees=2}
 {bone1Name=dyn_scarf_bk_03, bone2Name=dyn_scarf_bk_02, swing1LimitDegrees=15, swing2LimitDegrees=15, twistLimitDegrees=2}
 {bone1Name=dyn_scarf_bk_04, bone2Name=dyn_scarf_bk_03, swing1LimitDegrees=15, swing2LimitDegrees=25, twistLimitDegrees=2}
 {bone1Name=dyn_scarf_bk_05, bone2Name=dyn_scarf_bk_04, swing1LimitDegrees=15, swing2LimitDegrees=25, twistLimitDegrees=2}
 {
  bone1Name=base_scarf_fr
  bone2Name=spine_05
  swing1LimitDegrees=45
  swing1Motion=Free
  swing2LimitDegrees=45
  swing2Motion=Free
  twistLimitDegrees=45
  twistMotion=Free
 }
 {bone1Name=dyn_scarf_fr_01, bone2Name=base_scarf_fr, swing1LimitDegrees=25, swing2LimitDegrees=25, twistLimitDegrees=1}
 {bone1Name=dyn_scarf_fr_02, bone2Name=dyn_scarf_fr_01, swing1LimitDegrees=25, swing2LimitDegrees=25, twistLimitDegrees=1}
 {bone1Name=dyn_scarf_fr_03, bone2Name=dyn_scarf_fr_02, swing1LimitDegrees=25, swing2LimitDegrees=25, twistLimitDegrees=1}
 {
  bone1Name=clavicle_r
  bone2Name=spine_05
  swing1LimitDegrees=45
  swing1Motion=Free
  swing2LimitDegrees=45
  swing2Motion=Free
  twistLimitDegrees=45
  twistMotion=Free
 }
 {
  bone1Name=dyn_cloak_01
  bone2Name=spine_05
  swing1LimitDegrees=45
  swing1Motion=Free
  swing2LimitDegrees=45
  swing2Motion=Free
  twistLimitDegrees=45
  twistMotion=Free
 }
 {
  bone1Name=dyn_cloak_02
  bone2Name=dyn_cloak_01
  swing1LimitDegrees=45
  swing1Motion=Free
  swing2LimitDegrees=45
  swing2Motion=Free
  twistLimitDegrees=45
  twistMotion=Free
 }
]
