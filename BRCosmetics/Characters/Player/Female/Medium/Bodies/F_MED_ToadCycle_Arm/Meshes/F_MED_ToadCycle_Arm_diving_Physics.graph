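# Physics asset for the diving variant of the female medium ToadCycle_Arm body, covering its dynamic cape bones and lower-body collision shapes.
# Back, side and two front capsule chains per side hang from the pelvis, mostly with tight swing and twist limits.
# Free-rotation constraints cross-link neighboring chains into a sheet.
# Leg capsules, pelvis-locked cloth-volume capsules and a huge, floor-like root sphere supply body and ground shapes.
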PhysicsAsset /BRCosmetics/Characters/Player/Female/Medium/Bodies/F_MED_ToadCycle_Arm/Meshes/F_MED_ToadCycle_Arm_diving_Physics
bodies {
 calf_l [{shapeType=Capsule, length=35, radius=13.9819, shapeName=calf_l_capsule, center=<-22 -1.5008 1.6358>, rotation=<rot -90 -2.199 0>}]
 calf_r [{shapeType=Capsule, length=34, radius=14.4819, shapeName=calf_r_capsule, center=<22 1.767 -1.8707>, rotation=<rot -90 -2.199 0>}]
 dyn_Cloth_volume_l_A [{
   shapeType=Capsule
   length=51.255
   radius=8.755
   shapeName=dyn_Cloth_volume_r_B_capsule
   center=<-24.4017 12.0557 -5.4367>
   rotation=<rot 0.1242 -98.5735 -84.3299>
  }]
 dyn_Cloth_volume_l_B [{
   shapeType=Capsule
   length=51.255
   radius=8.755
   shapeName=dyn_Cloth_volume_r_B_capsule
   center=<-25.2715 18.1532 0.1096>
   rotation=<rot -87.0865 5.3339 171.7808>
  }]
 dyn_Cloth_volume_l_C [{
   shapeType=Capsule
   length=51.255
   radius=8.755
   shapeName=dyn_Cloth_volume_r_B_capsule
   center=<-26.2727 12.9279 5.1807>
   rotation=<rot -0.4036 75.8075 98.4184>
  }]
 dyn_Cloth_volume_r_A [{
   shapeType=Capsule
   length=51.255
   radius=8.755
   shapeName=dyn_Cloth_volume_r_B_capsule
   center=<-23.8831 11.9909 0.7013>
   rotation=<rot 0.1242 -98.5738 -98.3923>
  }]
 dyn_Cloth_volume_r_B [{
   shapeType=Capsule
   length=51.255
   radius=8.755
   shapeName=dyn_Cloth_volume_r_B_capsule
   center=<-25.6667 12.0058 0.0605>
   rotation=<rot 77.1716 -101.0433 -95.6897>
  }]
 dyn_Cloth_volume_r_C [{
   shapeType=Capsule
   length=51.255
   radius=8.755
   shapeName=dyn_Cloth_volume_r_B_capsule
   center=<-25.4994 12.678 -2.2973>
   rotation=<rot 7.8934 80.7277 81.7802>
  }]
 dyn_cape_b_l_01 [{shapeType=Capsule, length=11.5, radius=15.75, shapeName=dyn_cape_b_r_01_capsule, center=<4.922 0 8.1888>, rotation=<rot -0.0001 0 -90>}]
 dyn_cape_b_l_02 [{shapeType=Capsule, length=11.5, radius=15.75, shapeName=dyn_cape_b_r_02_capsule, center=<4.922 0 8.1888>, rotation=<rot -0.0001 0 -90>}]
 dyn_cape_b_l_03 [{shapeType=Capsule, length=21.6327, radius=21, shapeName=dyn_cape_b_r_04_capsule, center=<11.3668 0 11.7537>, rotation=<rot -0.0001 0 -90>}]
 dyn_cape_b_l_04 [{shapeType=Capsule, length=21.6327, radius=21, shapeName=dyn_cape_b_r_04_capsule, center=<11.3668 0 11.7537>, rotation=<rot -0.0001 0 -90>}]
 dyn_cape_b_l_base [{shapeType=Capsule, length=6.5, radius=1.505, shapeName=dyn_cape_b_l_base_capsule, center=<2.626 0 0>, rotation=<rot -0.0001 0 -90>}]
 dyn_cape_b_r_01 [{shapeType=Capsule, length=11.5, radius=15.75, shapeName=dyn_cape_b_r_01_capsule, center=<4.922 0 8.1888>, rotation=<rot -0.0001 0 -90>}]
 dyn_cape_b_r_02 [{shapeType=Capsule, length=11.5, radius=15.75, shapeName=dyn_cape_b_r_02_capsule, center=<4.922 0 8.1888>, rotation=<rot -0.0001 0 -90>}]
 dyn_cape_b_r_03 [{shapeType=Capsule, length=21.6327, radius=21, shapeName=dyn_cape_b_r_04_capsule, center=<11.3668 0 11.7537>, rotation=<rot -0.0001 0 -90>}]
 dyn_cape_b_r_04 [{shapeType=Capsule, length=21.6327, radius=21, shapeName=dyn_cape_b_r_04_capsule, center=<11.3668 0 11.7537>, rotation=<rot -0.0001 0 -90>}]
 dyn_cape_b_r_base [{shapeType=Capsule, length=6.5, radius=1.505, shapeName=dyn_cape_b_r_base_capsule, center=<2.626 0 0>, rotation=<rot -0.0001 0 -90>}]
 dyn_cape_f2_l_01 [{
   shapeType=Capsule
   length=21.25
   radius=17.75
   shapeName=dyn_cape_f2_r_03_capsule
   center=<20.0433 3.7433 12.3692>
   rotation=<rot 0.0005 -0.0002 -78.75>
  }]
 dyn_cape_f2_l_02 [{
   shapeType=Capsule
   length=20.25
   radius=21.5
   shapeName=dyn_cape_f2_r_03_capsule
   center=<16.0045 3.753 11.6719>
   rotation=<rot 0.0005 -0.0002 -78.75>
  }]
 dyn_cape_f2_l_03 [{
   shapeType=Capsule
   length=18.25
   radius=23.5
   shapeName=dyn_cape_f2_r_03_capsule
   center=<16.9365 3.721 13.1213>
   rotation=<rot 0.0005 -0.0002 -78.75>
  }]
 dyn_cape_f2_l_04 [{
   shapeType=Capsule
   length=16.5
   radius=25.25
   shapeName=dyn_cape_f2_r_03_capsule
   center=<16.3892 2.7851 11.3648>
   rotation=<rot 0.0005 -0.0002 -78.75>
  }]
 dyn_cape_f2_l_05 [{
   shapeType=Capsule
   length=19.5
   radius=22.25
   shapeName=dyn_cape_f2_r_03_capsule
   center=<16.3892 3.1312 12.7861>
   rotation=<rot 0.0005 -0.0002 -78.75>
  }]
 dyn_cape_f2_l_base [{shapeType=Capsule, length=6.5, radius=1.505, shapeName=dyn_cape_f2_l_base_capsule, center=<2.626 0 0>, rotation=<rot 0.0001 0.0001 -90>}]
 dyn_cape_f2_r_01 [{
   shapeType=Capsule
   length=21.25
   radius=17.75
   shapeName=dyn_cape_f2_r_03_capsule
   center=<20.0433 4.2677 15.0057>
   rotation=<rot 0.0005 -0.0002 -78.75>
  }]
 dyn_cape_f2_r_02 [{
   shapeType=Capsule
   length=20.25
   radius=21.5
   shapeName=dyn_cape_f2_r_03_capsule
   center=<20.0433 4.4428 15.1399>
   rotation=<rot 0.0005 -0.0002 -78.75>
  }]
 dyn_cape_f2_r_03 [{
   shapeType=Capsule
   length=20.75
   radius=21
   shapeName=dyn_cape_f2_r_03_capsule
   center=<20.0433 3.6501 12.7646>
   rotation=<rot 0.0005 -0.0002 -78.75>
  }]
 dyn_cape_f2_r_04 [{
   shapeType=Capsule
   length=18.75
   radius=23
   shapeName=dyn_cape_f2_r_03_capsule
   center=<16.3892 3.6839 15.8831>
   rotation=<rot 0.0005 -0.0002 -78.75>
  }]
 dyn_cape_f2_r_05 [{
   shapeType=Capsule
   length=18.75
   radius=23
   shapeName=dyn_cape_f2_r_03_capsule
   center=<16.3892 3.8261 16.2796>
   rotation=<rot 0.0005 -0.0002 -78.75>
  }]
 dyn_cape_f2_r_base [{shapeType=Capsule, length=6.5, radius=1.505, shapeName=dyn_cape_f2_r_base_capsule, center=<2.626 0 0>, rotation=<rot 0.0001 -0.0001 -90>}]
 dyn_cape_f_l_01 [{
   shapeType=Capsule
   length=21.5
   radius=19.25
   shapeName=dyn_cape_f2_r_03_capsule
   center=<12.4776 3.7185 10.8179>
   rotation=<rot 0.0003 -0.0003 -78.75>
  }]
 dyn_cape_f_l_02 [{
   shapeType=Capsule
   length=16.5
   radius=24.25
   shapeName=dyn_cape_f2_r_03_capsule
   center=<20.0433 3.5862 12.1487>
   rotation=<rot 0.0003 -0.0003 -78.75>
  }]
 dyn_cape_f_l_03 [{
   shapeType=Capsule
   length=17
   radius=23.75
   shapeName=dyn_cape_f2_r_03_capsule
   center=<20.0433 3.6116 14.7483>
   rotation=<rot 0.0003 -0.0003 -78.75>
  }]
 dyn_cape_f_l_04 [{
   shapeType=Capsule
   length=17.75
   radius=23
   shapeName=dyn_cape_f2_r_03_capsule
   center=<20.0432 4.4249 18.8373>
   rotation=<rot 0.0003 -0.0003 -78.75>
  }]
 dyn_cape_f_l_base [{shapeType=Capsule, length=6.5, radius=1.505, shapeName=dyn_cape_f_l_base_capsule, center=<2.626 0 0>, rotation=<rot 0 0.0001 -90>}]
 dyn_cape_f_r_01 [{
   shapeType=Capsule
   length=21.5
   radius=19.25
   shapeName=dyn_cape_f2_r_03_capsule
   center=<12.4776 4.1688 13.0821>
   rotation=<rot 0.0003 -0.0003 -78.75>
  }]
 dyn_cape_f_r_02 [{
   shapeType=Capsule
   length=16.5
   radius=24.25
   shapeName=dyn_cape_f2_r_03_capsule
   center=<20.0432 4.3572 16.0252>
   rotation=<rot 0.0003 -0.0003 -78.75>
  }]
 dyn_cape_f_r_03 [{
   shapeType=Capsule
   length=17
   radius=23.75
   shapeName=dyn_cape_f2_r_03_capsule
   center=<20.0433 3.7153 15.2697>
   rotation=<rot 0.0003 -0.0003 -78.75>
  }]
 dyn_cape_f_r_04 [{
   shapeType=Capsule
   length=17
   radius=23.75
   shapeName=dyn_cape_f2_r_03_capsule
   center=<20.0433 3.5014 14.1944>
   rotation=<rot 0.0003 -0.0003 -78.75>
  }]
 dyn_cape_f_r_base [{shapeType=Capsule, length=6.5, radius=1.505, shapeName=dyn_cape_f_r_base_capsule, center=<2.626 0 0>, rotation=<rot 0 -0.0001 -90>}]
 dyn_cape_s_l_01 [{shapeType=Capsule, length=12, radius=18, shapeName=dyn_cape_s_r_01_capsule, center=<8.9075 0 10.1779>, rotation=<rot -0.0001 -0.0001 -90>}]
 dyn_cape_s_l_02 [{shapeType=Capsule, length=12, radius=18, shapeName=dyn_cape_s_r_02_capsule, center=<8.9075 0 8.2769>, rotation=<rot -0.0001 -0.0001 -90>}]
 dyn_cape_s_l_03 [{
   shapeType=Capsule
   length=15.7122
   radius=18
   shapeName=dyn_cape_s_r_03_capsule
   center=<8.4714 0 11.4633>
   rotation=<rot -0.0001 -0.0001 -90>
  }]
 dyn_cape_s_l_04 [{
   shapeType=Capsule
   length=22.1191
   radius=18
   shapeName=dyn_cape_s_r_04_capsule
   center=<8.4714 0 11.4443>
   rotation=<rot -0.0001 -0.0001 -90>
  }]
 dyn_cape_s_l_base [{shapeType=Capsule, length=6.5, radius=1.505, shapeName=dyn_cape_s_l_base_capsule, center=<2.626 0 0>, rotation=<rot -0.0001 0.0001 -90>}]
 dyn_cape_s_r_01 [{shapeType=Capsule, length=12, radius=18, shapeName=dyn_cape_s_r_01_capsule, center=<8.9075 0 10.1779>, rotation=<rot -0.0001 -0.0001 -90>}]
 dyn_cape_s_r_02 [{shapeType=Capsule, length=12, radius=18, shapeName=dyn_cape_s_r_02_capsule, center=<8.9075 0 10.1779>, rotation=<rot -0.0001 -0.0001 -90>}]
 dyn_cape_s_r_03 [{
   shapeType=Capsule
   length=15.7122
   radius=18
   shapeName=dyn_cape_s_r_03_capsule
   center=<8.4714 0 13.3643>
   rotation=<rot -0.0001 -0.0001 -90>
  }]
 dyn_cape_s_r_04 [{
   shapeType=Capsule
   length=22.1191
   radius=18
   shapeName=dyn_cape_s_r_04_capsule
   center=<8.4714 0 13.3453>
   rotation=<rot -0.0001 -0.0001 -90>
  }]
 dyn_cape_s_r_base [{shapeType=Capsule, length=6.5, radius=1.505, shapeName=dyn_cape_s_r_base_capsule, center=<2.626 0 0>, rotation=<rot -0.0001 -0.0001 -90>}]
 foot_l [{shapeType=Capsule, length=8.7544, radius=13.7544, shapeName=foot_l_capsule, center=<-5 8 0>, rotation=<rot 0 0 -90>}]
 foot_r [{shapeType=Capsule, length=8.7544, radius=13.7544, shapeName=foot_r_capsule, center=<5 -8 0>, rotation=<rot 0 0 -90>}]
 pelvis [{shapeType=Capsule, length=10.3201, radius=10.3201, shapeName=pelvis_capsule}]
 root [{shapeType=Sphere, radius=500, shapeName=root_sphere, center=<0 17.4691 -493.7593>}]
 spine_01 [{shapeType=Capsule, length=10.3201, radius=10.3201, shapeName=spine_01_capsule}]
 thigh_l [{shapeType=Capsule, length=29.75, radius=6.9716, shapeName=thigh_l_capsule, center=<-25 -0.3962 3.4366>, rotation=<rot -90 0 0>}]
 thigh_r [{shapeType=Capsule, length=24.5, radius=12.2216, shapeName=thigh_r_capsule, center=<25 1.407 -2.2585>, rotation=<rot -84.375 -0.0001 0>}]
}
constraints [
 {
  bone1Name=pelvis
  bone2Name=root
  swing1LimitDegrees=45
  swing1Motion=Locked
  swing2LimitDegrees=45
  swing2Motion=Locked
  twistLimitDegrees=45
  twistMotion=Locked
 }
 {
  bone1Name=spine_01
  bone2Name=root
  swing1LimitDegrees=45
  swing1Motion=Free
  swing2LimitDegrees=45
  swing2Motion=Free
  twistLimitDegrees=45
  twistMotion=Free
 }
 {
  bone1Name=thigh_l
  bone2Name=root
  swing1LimitDegrees=45
  swing1Motion=Free
  swing2LimitDegrees=45
  swing2Motion=Free
  twistLimitDegrees=45
  twistMotion=Free
 }
 {
  bone1Name=calf_l
  bone2Name=root
  swing1LimitDegrees=45
  swing1Motion=Free
  swing2LimitDegrees=45
  swing2Motion=Free
  twistLimitDegrees=45
  twistMotion=Free
 }
 {
  bone1Name=thigh_r
  bone2Name=root
  swing1LimitDegrees=45
  swing1Motion=Free
  swing2LimitDegrees=45
  swing2Motion=Free
  twistLimitDegrees=45
  twistMotion=Free
 }
 {
  bone1Name=calf_r
  bone2Name=root
  swing1LimitDegrees=45
  swing1Motion=Free
  swing2LimitDegrees=45
  swing2Motion=Free
  twistLimitDegrees=45
  twistMotion=Free
 }
 {
  bone1Name=foot_l
  bone2Name=calf_l
  swing1LimitDegrees=45
  swing1Motion=Locked
  swing2LimitDegrees=45
  swing2Motion=Locked
  twistLimitDegrees=45
  twistMotion=Locked
 }
 {
  bone1Name=foot_r
  bone2Name=calf_r
  swing1LimitDegrees=45
  swing1Motion=Locked
  swing2LimitDegrees=45
  swing2Motion=Locked
  twistLimitDegrees=45
  twistMotion=Locked
 }
 {
  bone1Name=dyn_cape_b_l_base
  bone2Name=pelvis
  swing1LimitDegrees=45
  swing1Motion=Free
  swing2LimitDegrees=45
  swing2Motion=Free
  twistLimitDegrees=2
 }
 {bone1Name=dyn_cape_b_l_01, bone2Name=dyn_cape_b_l_base, swing1LimitDegrees=8, swing2LimitDegrees=8, twistLimitDegrees=2}
 {bone1Name=dyn_cape_b_l_02, bone2Name=dyn_cape_b_l_01, swing1LimitDegrees=8, swing2LimitDegrees=8, twistLimitDegrees=2}
 {bone1Name=dyn_cape_b_l_03, bone2Name=dyn_cape_b_l_02, swing1LimitDegrees=10, swing2LimitDegrees=25, twistLimitDegrees=2, twistMotion=Locked}
 {bone1Name=dyn_cape_b_l_04, bone2Name=dyn_cape_b_l_03, swing1LimitDegrees=10, swing2LimitDegrees=25, twistLimitDegrees=2, twistMotion=Locked}
 {
  bone1Name=dyn_cape_b_r_base
  bone2Name=pelvis
  swing1LimitDegrees=45
  swing1Motion=Locked
  swing2LimitDegrees=45
  swing2Motion=Locked
  twistLimitDegrees=2
 }
 {bone1Name=dyn_cape_b_r_01, bone2Name=dyn_cape_b_r_base, swing1LimitDegrees=8, swing2LimitDegrees=8, twistLimitDegrees=2}
 {bone1Name=dyn_cape_b_r_02, bone2Name=dyn_cape_b_r_01, swing1LimitDegrees=8, swing2LimitDegrees=8, twistLimitDegrees=2}
 {bone1Name=dyn_cape_b_r_03, bone2Name=dyn_cape_b_r_02, swing1LimitDegrees=10, swing2LimitDegrees=25, twistLimitDegrees=2, twistMotion=Locked}
 {bone1Name=dyn_cape_b_r_04, bone2Name=dyn_cape_b_r_03, swing1LimitDegrees=10, swing2LimitDegrees=25, twistLimitDegrees=2, twistMotion=Locked}
 {bone1Name=dyn_cape_s_r_04, bone2Name=dyn_cape_s_r_03, swing1LimitDegrees=10, swing2LimitDegrees=25, twistLimitDegrees=2}
 {
  bone1Name=dyn_cape_f2_l_base
  bone2Name=pelvis
  swing1LimitDegrees=45
  swing1Motion=Locked
  swing2LimitDegrees=45
  swing2Motion=Locked
  twistLimitDegrees=2
 }
 {bone1Name=dyn_cape_f2_l_01, bone2Name=dyn_cape_f2_l_base, swing1LimitDegrees=8, swing2LimitDegrees=8, twistLimitDegrees=35}
 {bone1Name=dyn_cape_f2_l_02, bone2Name=dyn_cape_f2_l_01, swing1LimitDegrees=8, swing2LimitDegrees=8, twistLimitDegrees=35}
 {bone1Name=dyn_cape_f2_l_03, bone2Name=dyn_cape_f2_l_02, swing1LimitDegrees=8, swing2LimitDegrees=8, twistLimitDegrees=15}
 {bone1Name=dyn_cape_f2_l_04, bone2Name=dyn_cape_f2_l_03, swing1LimitDegrees=10, swing2LimitDegrees=25, twistLimitDegrees=15}
 {
  bone1Name=dyn_cape_f2_l_05
  bone2Name=dyn_cape_f2_l_04
  swing1LimitDegrees=10
  swing2LimitDegrees=25
  twistLimitDegrees=2
  twistMotion=Locked
 }
 {
  bone1Name=dyn_cape_s_r_base
  bone2Name=pelvis
  swing1LimitDegrees=45
  swing1Motion=Locked
  swing2LimitDegrees=45
  swing2Motion=Locked
  twistLimitDegrees=2
 }
 {
  bone1Name=dyn_cape_f2_r_base
  bone2Name=pelvis
  swing1LimitDegrees=45
  swing1Motion=Locked
  swing2LimitDegrees=45
  swing2Motion=Locked
  twistLimitDegrees=2
 }
 {bone1Name=dyn_cape_f2_r_01, bone2Name=dyn_cape_f2_r_base, swing1LimitDegrees=8, swing2LimitDegrees=8, twistLimitDegrees=35}
 {bone1Name=dyn_cape_f2_r_02, bone2Name=dyn_cape_f2_r_01, swing1LimitDegrees=8, swing2LimitDegrees=8, twistLimitDegrees=35}
 {bone1Name=dyn_cape_f2_r_03, bone2Name=dyn_cape_f2_r_02, swing1LimitDegrees=8, swing2LimitDegrees=8, twistLimitDegrees=2}
 {bone1Name=dyn_cape_f2_r_04, bone2Name=dyn_cape_f2_r_03, swing1LimitDegrees=10, swing2LimitDegrees=25, twistLimitDegrees=2}
 {bone1Name=dyn_cape_f2_r_05, bone2Name=dyn_cape_f2_r_04, swing1LimitDegrees=10, swing2LimitDegrees=25, twistLimitDegrees=2}
 {bone1Name=dyn_cape_s_r_01, bone2Name=dyn_cape_s_r_base, swing1LimitDegrees=8, swing2LimitDegrees=8, twistLimitDegrees=2}
 {
  bone1Name=dyn_cape_f_l_base
  bone2Name=pelvis
  swing1LimitDegrees=45
  swing1Motion=Locked
  swing2LimitDegrees=45
  swing2Motion=Locked
  twistLimitDegrees=2
 }
 {bone1Name=dyn_cape_f_l_01, bone2Name=dyn_cape_f_l_base, swing1LimitDegrees=8, swing2LimitDegrees=8, twistLimitDegrees=2, twistMotion=Locked}
 {bone1Name=dyn_cape_f_l_02, bone2Name=dyn_cape_f_l_01, swing1LimitDegrees=8, swing2LimitDegrees=8, twistLimitDegrees=2}
 {bone1Name=dyn_cape_f_l_03, bone2Name=dyn_cape_f_l_02, swing1LimitDegrees=10, swing2LimitDegrees=25, twistLimitDegrees=2}
 {bone1Name=dyn_cape_f_l_04, bone2Name=dyn_cape_f_l_03, swing1LimitDegrees=10, swing2LimitDegrees=25, twistLimitDegrees=2, twistMotion=Locked}
 {bone1Name=dyn_cape_s_r_02, bone2Name=dyn_cape_s_r_01, swing1LimitDegrees=8, swing2LimitDegrees=8, twistLimitDegrees=2}
 {
  bone1Name=dyn_cape_f_r_base
  bone2Name=pelvis
  swing1LimitDegrees=45
  swing1Motion=Locked
  swing2LimitDegrees=45
  swing2Motion=Locked
  twistLimitDegrees=2
 }
 {bone1Name=dyn_cape_f_r_01, bone2Name=dyn_cape_f_r_base, swing1LimitDegrees=8, swing2LimitDegrees=8, twistLimitDegrees=2}
 {bone1Name=dyn_cape_f_r_02, bone2Name=dyn_cape_f_r_01, swing1LimitDegrees=8, swing2LimitDegrees=8, twistLimitDegrees=2}
 {bone1Name=dyn_cape_f_r_03, bone2Name=dyn_cape_f_r_02, swing1LimitDegrees=10, swing2LimitDegrees=25, twistLimitDegrees=2}
 {bone1Name=dyn_cape_f_r_04, bone2Name=dyn_cape_f_r_03, swing1LimitDegrees=10, swing2LimitDegrees=25, twistLimitDegrees=2}
 {bone1Name=dyn_cape_s_r_03, bone2Name=dyn_cape_s_r_02, swing1LimitDegrees=10, swing2LimitDegrees=25, twistLimitDegrees=2}
 {
  bone1Name=dyn_cape_s_l_base
  bone2Name=pelvis
  swing1LimitDegrees=45
  swing1Motion=Locked
  swing2LimitDegrees=45
  swing2Motion=Locked
  twistLimitDegrees=2
 }
 {bone1Name=dyn_cape_s_l_01, bone2Name=dyn_cape_s_l_base, swing1LimitDegrees=8, swing2LimitDegrees=8, twistLimitDegrees=2}
 {bone1Name=dyn_cape_s_l_02, bone2Name=dyn_cape_s_l_01, swing1LimitDegrees=8, swing2LimitDegrees=8, twistLimitDegrees=2}
 {bone1Name=dyn_cape_s_l_03, bone2Name=dyn_cape_s_l_02, swing1LimitDegrees=10, swing2LimitDegrees=25, twistLimitDegrees=2}
 {bone1Name=dyn_cape_s_l_04, bone2Name=dyn_cape_s_l_03, swing1LimitDegrees=10, swing2LimitDegrees=25, twistLimitDegrees=2}
 {
  bone1Name=dyn_cape_b_r_04
  bone2Name=dyn_cape_b_l_04
  swing1LimitDegrees=45
  swing1Motion=Free
  swing2LimitDegrees=45
  swing2Motion=Free
  twistLimitDegrees=45
  twistMotion=Free
 }
 {
  bone1Name=dyn_cape_b_l_04
  bone2Name=dyn_cape_b_r_04
  swing1LimitDegrees=45
  swing1Motion=Free
  swing2LimitDegrees=45
  swing2Motion=Free
  twistLimitDegrees=45
  twistMotion=Free
 }
 {
  bone1Name=dyn_Cloth_volume_l_A
  bone2Name=pelvis
  swing1LimitDegrees=45
  swing1Motion=Locked
  swing2LimitDegrees=45
  swing2Motion=Locked
  twistLimitDegrees=45
  twistMotion=Locked
 }
 {
  bone1Name=dyn_Cloth_volume_l_B
  bone2Name=pelvis
  swing1LimitDegrees=45
  swing1Motion=Locked
  swing2LimitDegrees=45
  swing2Motion=Locked
  twistLimitDegrees=45
  twistMotion=Locked
 }
 {
  bone1Name=dyn_Cloth_volume_l_C
  bone2Name=pelvis
  swing1LimitDegrees=45
  swing1Motion=Locked
  swing2LimitDegrees=45
  swing2Motion=Locked
  twistLimitDegrees=45
  twistMotion=Locked
 }
 {
  bone1Name=dyn_Cloth_volume_r_A
  bone2Name=pelvis
  swing1LimitDegrees=45
  swing1Motion=Locked
  swing2LimitDegrees=45
  swing2Motion=Locked
  twistLimitDegrees=45
  twistMotion=Locked
 }
 {
  bone1Name=dyn_Cloth_volume_r_B
  bone2Name=pelvis
  swing1LimitDegrees=45
  swing1Motion=Locked
  swing2LimitDegrees=45
  swing2Motion=Locked
  twistLimitDegrees=45
  twistMotion=Locked
 }
 {
  bone1Name=dyn_Cloth_volume_r_C
  bone2Name=pelvis
  swing1LimitDegrees=45
  swing1Motion=Locked
  swing2LimitDegrees=45
  swing2Motion=Locked
  twistLimitDegrees=45
  twistMotion=Locked
 }
 {
  bone1Name=dyn_cape_f_r_03
  bone2Name=dyn_cape_s_r_03
  swing1LimitDegrees=45
  swing1Motion=Free
  swing2LimitDegrees=45
  swing2Motion=Free
  twistLimitDegrees=45
  twistMotion=Free
 }
 {
  bone1Name=dyn_cape_f_l_03
  bone2Name=dyn_cape_s_l_03
  swing1LimitDegrees=45
  swing1Motion=Free
  swing2LimitDegrees=45
  swing2Motion=Free
  twistLimitDegrees=45
  twistMotion=Free
 }
 {
  bone1Name=dyn_cape_f2_r_04
  bone2Name=dyn_cape_f_r_03
  swing1LimitDegrees=45
  swing1Motion=Free
  swing2LimitDegrees=45
  swing2Motion=Free
  twistLimitDegrees=45
  twistMotion=Free
 }
 {
  bone1Name=dyn_cape_f2_l_04
  bone2Name=dyn_cape_f_l_03
  swing1LimitDegrees=45
  swing1Motion=Free
  swing2LimitDegrees=45
  swing2Motion=Free
  twistLimitDegrees=45
  twistMotion=Free
 }
 {
  bone1Name=dyn_cape_s_l_03
  bone2Name=dyn_cape_b_l_03
  swing1LimitDegrees=45
  swing1Motion=Free
  swing2LimitDegrees=45
  swing2Motion=Free
  twistLimitDegrees=45
  twistMotion=Free
 }
 {
  bone1Name=dyn_cape_s_r_03
  bone2Name=dyn_cape_b_r_03
  swing1LimitDegrees=45
  swing1Motion=Free
  swing2LimitDegrees=45
  swing2Motion=Free
  twistLimitDegrees=45
  twistMotion=Free
 }
 {
  bone1Name=dyn_cape_f2_l_05
  bone2Name=dyn_cape_f_l_04
  swing1LimitDegrees=45
  swing1Motion=Free
  swing2LimitDegrees=45
  swing2Motion=Free
  twistLimitDegrees=45
  twistMotion=Free
 }
 {
  bone1Name=dyn_cape_b_r_03
  bone2Name=dyn_cape_b_l_03
  swing1LimitDegrees=45
  swing1Motion=Free
  swing2LimitDegrees=45
  swing2Motion=Free
  twistLimitDegrees=45
  twistMotion=Free
 }
 {
  bone1Name=dyn_cape_b_l_03
  bone2Name=dyn_cape_b_r_03
  swing1LimitDegrees=45
  swing1Motion=Free
  swing2LimitDegrees=45
  swing2Motion=Free
  twistLimitDegrees=45
  twistMotion=Free
 }
 {
  bone1Name=dyn_cape_s_l_04
  bone2Name=dyn_cape_b_l_04
  swing1LimitDegrees=45
  swing1Motion=Free
  swing2LimitDegrees=45
  swing2Motion=Free
  twistLimitDegrees=45
  twistMotion=Free
 }
 {
  bone1Name=dyn_cape_s_r_04
  bone2Name=dyn_cape_b_r_04
  swing1LimitDegrees=45
  swing1Motion=Free
  swing2LimitDegrees=45
  swing2Motion=Free
  twistLimitDegrees=45
  twistMotion=Free
 }
 {
  bone1Name=dyn_cape_f_l_04
  bone2Name=dyn_cape_s_l_04
  swing1LimitDegrees=45
  swing1Motion=Free
  swing2LimitDegrees=45
  swing2Motion=Free
  twistLimitDegrees=45
  twistMotion=Free
 }
 {
  bone1Name=dyn_cape_f_r_04
  bone2Name=dyn_cape_s_r_04
  swing1LimitDegrees=45
  swing1Motion=Free
  swing2LimitDegrees=45
  swing2Motion=Free
  twistLimitDegrees=45
  twistMotion=Free
 }
 {
  bone1Name=dyn_cape_f2_r_05
  bone2Name=dyn_cape_f_r_04
  swing1LimitDegrees=45
  swing1Motion=Free
  swing2LimitDegrees=45
  swing2Motion=Free
  twistLimitDegrees=45
  twistMotion=Free
 }
 {
  bone1Name=dyn_cape_f2_l_05
  bone2Name=dyn_cape_f_l_04
  swing1LimitDegrees=45
  swing1Motion=Free
  swing2LimitDegrees=45
  swing2Motion=Free
  twistLimitDegrees=45
  twistMotion=Free
 }
 {
  bone1Name=dyn_cape_s_r_02
  bone2Name=dyn_cape_b_r_02
  swing1LimitDegrees=45
  swing1Motion=Free
  swing2LimitDegrees=45
  swing2Motion=Free
  twistLimitDegrees=45
  twistMotion=Free
 }
 {
  bone1Name=dyn_cape_s_l_02
  bone2Name=dyn_cape_b_l_02
  swing1LimitDegrees=45
  swing1Motion=Free
  swing2LimitDegrees=45
  swing2Motion=Free
  twistLimitDegrees=45
  twistMotion=Free
 }
 {
  bone1Name=dyn_cape_f_l_02
  bone2Name=dyn_cape_s_l_02
  swing1LimitDegrees=45
  swing1Motion=Free
  swing2LimitDegrees=45
  swing2Motion=Free
  twistLimitDegrees=45
  twistMotion=Free
 }
 {
  bone1Name=dyn_cape_f_r_02
  bone2Name=dyn_cape_s_r_02
  swing1LimitDegrees=45
  swing1Motion=Free
  swing2LimitDegrees=45
  swing2Motion=Free
  twistLimitDegrees=45
  twistMotion=Free
 }
 {
  bone1Name=dyn_cape_f2_r_03
  bone2Name=dyn_cape_f_r_02
  swing1LimitDegrees=45
  swing1Motion=Free
  swing2LimitDegrees=45
  swing2Motion=Free
  twistLimitDegrees=45
  twistMotion=Free
 }
 {
  bone1Name=dyn_cape_f2_l_03
  bone2Name=dyn_cape_f_l_02
  swing1LimitDegrees=45
  swing1Motion=Free
  swing2LimitDegrees=45
  swing2Motion=Free
  twistLimitDegrees=45
  twistMotion=Free
 }
 {
  bone1Name=dyn_cape_f2_l_02
  bone2Name=dyn_cape_f_l_01
  swing1LimitDegrees=45
  swing1Motion=Free
  swing2LimitDegrees=45
  swing2Motion=Free
  twistLimitDegrees=45
  twistMotion=Free
 }
]
