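# Anim Blueprint for the F_MED_ToadCycle_Arm female cosmetic body, adding secondary motion to its cape, cloak, sleeves, scarf and cloth.
# The parent mesh pose passes through Control Rig and Anim Dynamics into a cache.
# Cloth, rigid-body, skydiving and DBNO flags select that pose, RigidBody or Trail controller variants.
# Sleeve offsets, further dynamics, upper-body RigidBody and spine constraints follow, with thigh-weighted cloth constraints added when skydiving.
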AnimBlueprint /BRCosmetics/Characters/Player/Female/Medium/Bodies/F_MED_ToadCycle_Arm/Meshes/F_MED_ToadCycle_Arm_AnimBP
target_skeleton=&/F_MED_ToadCycle_Arm_Skeleton
graphs [
 AnimGraph {nodes [
   n1 = "Constraint - dyn_Cloth_volume_l_A01 (PP)"!(ComponentPose=Local To Component(LocalPose=Use cached pose 'for diving'()), Alpha=1, ConstraintWeights_0=0.3, ConstraintWeights_1:float=0.7) -> (Pose) {
    boneToModify {boneName=dyn_Cloth_volume_l_A01}
    constraintSetup [{targetBone {boneName=thigh_l}}, {targetBone {boneName=dyn_Cloth_volume_l_A}}]
    constraintWeights [1,1]
   }
   n2 = "Constraint - dyn_Cloth_volume_r_A01 (PP)"!(ComponentPose=@n1, Alpha=1, ConstraintWeights_0=0.3, ConstraintWeights_1:float=0.7) -> (Pose) {
    boneToModify {boneName=dyn_Cloth_volume_r_A01}
    constraintSetup [{targetBone {boneName=thigh_r}}, {targetBone {boneName=dyn_Cloth_volume_r_A}}]
    constraintWeights [1,1]
   }
   n3 = "Constraint - dyn_Cloth_volume_l_B01 (PP)"!(ComponentPose=@n2, Alpha=1, ConstraintWeights_0=0.3, ConstraintWeights_1:float=0.7) -> (Pose) {
    boneToModify {boneName=dyn_Cloth_volume_l_B01}
    constraintSetup [{targetBone {boneName=thigh_l}}, {targetBone {boneName=dyn_Cloth_volume_l_B}}]
    constraintWeights [1,1]
   }
   n4 = "Constraint - dyn_Cloth_volume_r_B01 (PP)"!(ComponentPose=@n3, Alpha=1, ConstraintWeights_0=0.3, ConstraintWeights_1:float=0.7) -> (Pose) {
    boneToModify {boneName=dyn_Cloth_volume_r_B01}
    constraintSetup [{targetBone {boneName=thigh_r}}, {targetBone {boneName=dyn_Cloth_volume_r_B}}]
    constraintWeights [1,1]
   }
   n5 = "Constraint - dyn_Cloth_volume_l_C01 (PP)"!(ComponentPose=@n4, Alpha=1, ConstraintWeights_0=0.3, ConstraintWeights_1:float=0.7) -> (Pose) {
    boneToModify {boneName=dyn_Cloth_volume_l_C01}
    constraintSetup [{targetBone {boneName=thigh_l}}, {targetBone {boneName=dyn_Cloth_volume_l_C}}]
    constraintWeights [1,1]
   }
   n6 = "Constraint - dyn_Cloth_volume_r_C01 (PP)"!(ComponentPose=@n5, Alpha=1, ConstraintWeights_0=0.3, ConstraintWeights_1:float=0.7) -> (Pose) {
    boneToModify {boneName=dyn_Cloth_volume_r_C01}
    constraintSetup [{targetBone {boneName=thigh_r}}, {targetBone {boneName=dyn_Cloth_volume_r_C}}]
    constraintWeights [1,1]
   }
   n7 = "Constraint - dyn_Cloth_volume_bk01 (PPP)"!(ComponentPose=@n6, Alpha=1, ConstraintWeights_0=0.15, ConstraintWeights_1:float=0.7, ConstraintWeights_2:float=0.15) -> (Pose) {
    boneToModify {boneName=dyn_Cloth_volume_bk01}
    constraintSetup [{targetBone {boneName=thigh_l}}, {targetBone {boneName=dyn_Cloth_volume_bk}}, {targetBone {boneName=thigh_r}}]
    constraintWeights [1,1,1]
   }
   n8 = "Constraint - dyn_Cloth_volume_fn01 (PPT)"!(ComponentPose=@n7, Alpha=1, ConstraintWeights_0=0.15, ConstraintWeights_1:float=0.7, ConstraintWeights_2:float=0.15) -> (Pose) {
    boneToModify {boneName=dyn_Cloth_volume_fn01}
    constraintSetup [{targetBone {boneName=thigh_l}}, {targetBone {boneName=dyn_Cloth_volume_fn}}, {transformType=Translation, targetBone {boneName=thigh_r}}]
    constraintWeights [1,1,1]
   }
   Output Pose(Result=Blend Poses by bool(BlendPose_0=Component To Local(ComponentPose=@n8), BlendPose_1=Use cached pose 'for diving'(), bActiveValue=Get bIsSkyDiving()) {blendType=Linear}) {layerGroup=None}
   n9 = Blend Poses by bool(BlendPose_0=Use cached pose 'dbnoCache'(), BlendPose_1=Use cached pose 'trailCache'(), bActiveValue=Get bIsDBNO()) {blendType=Linear}
   n10 = Blend Poses by bool(BlendPose_0=Use cached pose 'ctrlRig_cache'(), BlendPose_1=Use cached pose 'physicsCape_Cache'(), bActiveValue=Get bClothEnabled()) {blendType=Linear}
   n11 = Get bRigidBodyEnabled()
   n12 = "Transform (Modify) Bone - Bone: base_sleeveTop_l"(
    ComponentPose=Local To Component(LocalPose=Blend Poses by bool(BlendPose_0=@n10, BlendPose_1=@n9, bActiveValue=@n11) {blendType=Linear})
    Alpha=1
    Translation="0,0,0"
    Rotation="-2.543,0.622,4.608"
    Scale="1,1,1"
   ) {
    rotationMode=BMM_Additive
    rotationSpace=BCS_BoneSpace
    translationMode=BMM_Additive
    translationSpace=BCS_BoneSpace
    alphaBoolBlend {blendOption=Linear}
    boneToModify {boneName=base_sleeveTop_l}
    rotation=<rot -2.543 0.622 4.608>
   }
   n13 = "Transform (Modify) Bone - Bone: base_sleeveBottom_l"(ComponentPose=@n12, Alpha=1, Translation="-0.415,1.375,-0.405", Rotation="6.756,11.837,6.685", Scale="1,1,1") {
    rotationMode=BMM_Additive
    rotationSpace=BCS_BoneSpace
    translationMode=BMM_Additive
    translationSpace=BCS_BoneSpace
    alphaBoolBlend {blendOption=Linear}
    boneToModify {boneName=base_sleeveBottom_l}
    rotation=<rot 6.756 11.837 6.685>
    translation=<-0.415 1.375 -0.405>
   }
   n14 = "Transform (Modify) Bone - Bone: base_sleeveBack_l"(ComponentPose=@n13, Alpha=1, Translation="0,0,0", Rotation="4.080,8.355,-6.667", Scale="1,1,1") {
    rotationMode=BMM_Additive
    rotationSpace=BCS_BoneSpace
    translationMode=BMM_Additive
    translationSpace=BCS_BoneSpace
    alphaBoolBlend {blendOption=Linear}
    boneToModify {boneName=base_sleeveBack_l}
    rotation=<rot 4.0801 8.355 -6.667>
   }
   n15 = "Transform (Modify) Bone - Bone: base_sleeveFront_l"(ComponentPose=@n14, Alpha=1, Translation="0,0,0", Rotation="5.137,-0.262,0.161", Scale="1,1,1") {
    rotationMode=BMM_Additive
    rotationSpace=BCS_BoneSpace
    translationMode=BMM_Additive
    translationSpace=BCS_BoneSpace
    alphaBoolBlend {blendOption=Linear}
    boneToModify {boneName=base_sleeveFront_l}
    rotation=<rot 5.137 -0.262 0.161>
   }
   n16 = "Transform (Modify) Bone - Bone: base_sleeveTop_r"(ComponentPose=@n15, Alpha=1, Translation="1.704,0.188,-0.083", Rotation="-0.960,6.851,12.920", Scale="1,1,1") {
    rotationMode=BMM_Additive
    rotationSpace=BCS_BoneSpace
    translationMode=BMM_Additive
    translationSpace=BCS_BoneSpace
    alphaBoolBlend {blendOption=Linear}
    boneToModify {boneName=base_sleeveTop_r}
    rotation=<rot -0.9599 6.851 12.92>
    translation=<1.704 0.188 -0.083>
   }
   n17 = "Transform (Modify) Bone - Bone: base_sleeveBottom_r"(ComponentPose=@n16, Alpha=1, Translation="0,0,0", Rotation="-0.876,18.300,-19.262", Scale="1,1,1") {
    rotationMode=BMM_Additive
    rotationSpace=BCS_BoneSpace
    translationMode=BMM_Additive
    translationSpace=BCS_BoneSpace
    alphaBoolBlend {blendOption=Linear}
    boneToModify {boneName=base_sleeveBottom_r}
    rotation=<rot -0.8761 18.3 -19.2621>
   }
   n18 = "Transform (Modify) Bone - Bone: base_sleeveBack_r"(ComponentPose=@n17, Alpha=1, Translation="0,0,0", Rotation="5.422,1.573,1.495", Scale="1,1,1") {
    rotationMode=BMM_Additive
    rotationSpace=BCS_BoneSpace
    translationMode=BMM_Additive
    translationSpace=BCS_BoneSpace
    alphaBoolBlend {blendOption=Linear}
    boneToModify {boneName=base_sleeveBack_r}
    rotation=<rot 5.422 1.573 1.495>
   }
   n19 = "Transform (Modify) Bone - Bone: base_sleeveFront_r"(ComponentPose=@n18, Alpha=1, Translation="0,0,0", Rotation="-7.108,6.200,-0.989", Scale="1,1,1") {
    rotationMode=BMM_Additive
    rotationSpace=BCS_BoneSpace
    translationMode=BMM_Additive
    translationSpace=BCS_BoneSpace
    alphaBoolBlend {blendOption=Linear}
    boneToModify {boneName=base_sleeveFront_r}
    rotation=<rot -7.108 6.2 -0.989>
   }
   n20 = Anim Dynamics - Bone: dyn_sleeveTop_l(ComponentPose=@n19, Alpha=1, ExternalForce="0,0,0") {
    angularDampingOverride=1
    angularSpringConstant=1000
    bAngularSpring=true
    bOverrideAngularDamping=true
    bUseGravityOverride=true
    gravityScale=0.35
    simulationSpace=RootRelative
    alphaBoolBlend {blendOption=Linear}
    boundBone {boneName=dyn_sleeveTop_l}
    componentAppliedLinearAccClamp=<2000 2000 2000>
    componentLinearAccScale=<0.4 0.4 0.4>
    componentLinearVelScale=<0.5 0.5 0.5>
    gravityOverride {z=-980}
    physicsBodyDefinitions [{
      boundBone {boneName=dyn_sleeveTop_l}
      boxExtents=<17 17 17>
      constraintSetup {angularLimitsMax=<5 0 2>, angularLimitsMin=<-5 0 -15>, angularTarget=<1 0 0>}
      localJointOffset=<2 0 0>
     }]
    simSpaceSettings {maxAngularAcceleration=40, maxAngularVelocity=8, simSpaceAngularAlpha=1}
   }
   n21 = Anim Dynamics - Bone: dyn_sleeveBottom_l(ComponentPose=@n20, Alpha=1, ExternalForce="0,0,0") {
    angularDampingOverride=1
    angularSpringConstant=1000
    bAngularSpring=true
    bOverrideAngularDamping=true
    bUseGravityOverride=true
    gravityScale=0.35
    simulationSpace=RootRelative
    alphaBoolBlend {blendOption=Linear}
    boundBone {boneName=dyn_sleeveBottom_l}
    componentAppliedLinearAccClamp=<2000 2000 2000>
    componentLinearAccScale=<0.4 0.4 0.4>
    componentLinearVelScale=<0.5 0.5 0.5>
    gravityOverride {z=-980}
    physicsBodyDefinitions [{
      boundBone {boneName=dyn_sleeveBottom_l}
      boxExtents=<17 17 17>
      constraintSetup {angularLimitsMax=<10 5 8>, angularLimitsMin=<-10 -5 -12>, angularTarget=<1 0 0>}
      localJointOffset=<2 0 0>
     }]
    simSpaceSettings {maxAngularAcceleration=40, maxAngularVelocity=8, simSpaceAngularAlpha=1}
   }
   n22 = Anim Dynamics - Bone: dyn_sleeveFront_l(ComponentPose=@n21, Alpha=1, ExternalForce="0,0,0") {
    angularDampingOverride=1
    angularSpringConstant=1000
    bAngularSpring=true
    bOverrideAngularDamping=true
    bUseGravityOverride=true
    gravityScale=0.35
    simulationSpace=RootRelative
    alphaBoolBlend {blendOption=Linear}
    boundBone {boneName=dyn_sleeveFront_l}
    componentAppliedLinearAccClamp=<2000 2000 2000>
    componentLinearAccScale=<0.4 0.4 0.4>
    componentLinearVelScale=<0.5 0.5 0.5>
    gravityOverride {z=-980}
    physicsBodyDefinitions [{
      boundBone {boneName=dyn_sleeveFront_l}
      boxExtents=<17 17 17>
      constraintSetup {angularLimitsMax=<5 0 4>, angularLimitsMin=<-5 0 -26>, angularTarget=<1 0 0>}
      localJointOffset=<2 0 0>
     }]
    simSpaceSettings {maxAngularAcceleration=40, maxAngularVelocity=8, simSpaceAngularAlpha=1}
   }
   n23 = Anim Dynamics - Bone: dyn_sleeveBack_l(ComponentPose=@n22, Alpha=1, ExternalForce="0,0,0") {
    angularDampingOverride=1
    angularSpringConstant=1000
    bAngularSpring=true
    bOverrideAngularDamping=true
    bUseGravityOverride=true
    gravityScale=0.35
    simulationSpace=RootRelative
    alphaBoolBlend {blendOption=Linear}
    boundBone {boneName=dyn_sleeveBack_l}
    componentAppliedLinearAccClamp=<2000 2000 2000>
    componentLinearAccScale=<0.4 0.4 0.4>
    componentLinearVelScale=<0.5 0.5 0.5>
    gravityOverride {z=-980}
    physicsBodyDefinitions [{
      boundBone {boneName=dyn_sleeveBack_l}
      boxExtents=<17 17 17>
      constraintSetup {angularLimitsMax=<5 8 2>, angularLimitsMin=<-5 -8 -16>, angularTarget=<1 0 0>}
      localJointOffset=<2 0 0>
     }]
    simSpaceSettings {maxAngularAcceleration=40, maxAngularVelocity=8, simSpaceAngularAlpha=1}
   }
   n24 = Anim Dynamics - Bone: dyn_sleeveTop_r(ComponentPose=@n23, Alpha=1, ExternalForce="0,0,0") {
    angularDampingOverride=1
    angularSpringConstant=1000
    bAngularSpring=true
    bOverrideAngularDamping=true
    bUseGravityOverride=true
    gravityScale=0.35
    simulationSpace=RootRelative
    alphaBoolBlend {blendOption=Linear}
    boundBone {boneName=dyn_sleeveTop_r}
    componentAppliedLinearAccClamp=<2000 2000 2000>
    componentLinearAccScale=<0.4 0.4 0.4>
    componentLinearVelScale=<0.5 0.5 0.5>
    gravityOverride {z=-980}
    physicsBodyDefinitions [{
      boundBone {boneName=dyn_sleeveTop_r}
      boxExtents=<17 17 17>
      constraintSetup {angularLimitsMax=<5 0 2>, angularLimitsMin=<-5 0 -13>, angularTarget=<1 0 0>}
      localJointOffset=<-2 0 0>
     }]
    simSpaceSettings {maxAngularAcceleration=40, maxAngularVelocity=8, simSpaceAngularAlpha=1}
   }
   n25 = Anim Dynamics - Bone: dyn_sleeveBottom_r(ComponentPose=@n24, Alpha=1, ExternalForce="0,0,0") {
    angularDampingOverride=1
    angularSpringConstant=1000
    bAngularSpring=true
    bOverrideAngularDamping=true
    bUseGravityOverride=true
    gravityScale=0.35
    simulationSpace=RootRelative
    alphaBoolBlend {blendOption=Linear}
    boundBone {boneName=dyn_sleeveBottom_r}
    componentAppliedLinearAccClamp=<2000 2000 2000>
    componentLinearAccScale=<0.4 0.4 0.4>
    componentLinearVelScale=<0.5 0.5 0.5>
    gravityOverride {z=-980}
    physicsBodyDefinitions [{
      boundBone {boneName=dyn_sleeveBottom_r}
      boxExtents=<17 17 17>
      constraintSetup {angularLimitsMax=<10 5 8>, angularLimitsMin=<-10 -5 -12>, angularTarget=<1 0 0>}
      localJointOffset=<-2 0 0>
     }]
    simSpaceSettings {maxAngularAcceleration=40, maxAngularVelocity=8, simSpaceAngularAlpha=1}
   }
   n26 = Anim Dynamics - Bone: dyn_sleeveFront_r(ComponentPose=@n25, Alpha=1, ExternalForce="0,0,0") {
    angularDampingOverride=1
    angularSpringConstant=1000
    bAngularSpring=true
    bOverrideAngularDamping=true
    bUseGravityOverride=true
    gravityScale=0.35
    simulationSpace=RootRelative
    alphaBoolBlend {blendOption=Linear}
    boundBone {boneName=dyn_sleeveFront_r}
    componentAppliedLinearAccClamp=<2000 2000 2000>
    componentLinearAccScale=<0.4 0.4 0.4>
    componentLinearVelScale=<0.5 0.5 0.5>
    gravityOverride {z=-980}
    physicsBodyDefinitions [{
      boundBone {boneName=dyn_sleeveFront_r}
      boxExtents=<17 17 17>
      constraintSetup {angularLimitsMax=<8 0 3>, angularLimitsMin=<-8 0 -20>, angularTarget=<1 0 0>}
      localJointOffset=<-2 0 0>
     }]
    simSpaceSettings {maxAngularAcceleration=40, maxAngularVelocity=8, simSpaceAngularAlpha=1}
   }
   n27 = Anim Dynamics - Bone: dyn_sleeveBack_r(ComponentPose=@n26, Alpha=1, ExternalForce="0,0,0") {
    angularDampingOverride=1
    angularSpringConstant=1000
    bAngularSpring=true
    bOverrideAngularDamping=true
    bUseGravityOverride=true
    gravityScale=0.35
    simulationSpace=RootRelative
    alphaBoolBlend {blendOption=Linear}
    boundBone {boneName=dyn_sleeveBack_r}
    componentAppliedLinearAccClamp=<2000 2000 2000>
    componentLinearAccScale=<0.4 0.4 0.4>
    componentLinearVelScale=<0.5 0.5 0.5>
    gravityOverride {z=-980}
    physicsBodyDefinitions [{
      boundBone {boneName=dyn_sleeveBack_r}
      boxExtents=<17 17 17>
      constraintSetup {angularLimitsMax=<8 8 6>, angularLimitsMin=<-8 -8 -16>, angularTarget=<1 0 0>}
      localJointOffset=<-2 0 0>
     }]
    simSpaceSettings {maxAngularAcceleration=40, maxAngularVelocity=8, simSpaceAngularAlpha=1}
   }
   n28 = "Constraint - dyn_cloak_01 (PPP)"!(ComponentPose=@n27, Alpha=1, ConstraintWeights_0=0.5339, ConstraintWeights_1:float=0.4409, ConstraintWeights_2:float=0.0252) -> (Pose) {
    boneToModify {boneName=dyn_cloak_01}
    constraintSetup [{targetBone {boneName=spine_04}}, {targetBone {boneName=dyn_cloak_01}}, {targetBone {boneName=clavicle_r}}]
    constraintWeights [1,1,1]
   }
   n29 = "Constraint - dyn_cloak_02 (P)"!(ComponentPose=@n28, Alpha=1, ConstraintWeights_0=1, ConstraintWeights_1:float=0) -> (Pose) {boneToModify {boneName=dyn_cloak_02}, constraintSetup [{targetBone {boneName=spine_04}}], constraintWeights [1,0]}
   n30 = Anim Dynamics - Bone: dyn_cloak_01!(ComponentPose=@n29, Alpha=1, GravityOverride:Vector="0,0,-980", ExternalForce="0,0,0") -> (Pose) {
    angularDampingOverride=1
    angularSpringConstant=800
    bAngularSpring=true
    bOverrideAngularDamping=true
    bUseGravityOverride=true
    gravityScale=0.3
    simulationSpace=RootRelative
    alphaBoolBlend {blendOption=Linear}
    boundBone {boneName=dyn_cloak_01}
    componentAppliedLinearAccClamp=<2000 2000 1600>
    componentLinearAccScale=<0.4 0.4 0.4>
    componentLinearVelScale=<0.45 0.45 0.55>
    gravityOverride {z=-980}
    physicsBodyDefinitions [{
      sphereCollisionRadius=2
      boundBone {boneName=dyn_cloak_01}
      boxExtents=<10 16 16>
      constraintSetup {angularLimitsMax=<0 0 20>, angularLimitsMin=<0 -45 -20>, angularTarget=<0 0 0>}
      localJointOffset=<2 0 -0.1>
     }]
    relativeSpaceBone {boneName=root}
   }
   n31 = Anim Dynamics - Bone: dyn_cloak_02!(ComponentPose=@n30, Alpha=1, GravityOverride:Vector="0,0,-980", ExternalForce="0,0,0") -> (Pose) {
    angularDampingOverride=1
    angularSpringConstant=800
    bAngularSpring=true
    bOverrideAngularDamping=true
    bUseGravityOverride=true
    gravityScale=0.3
    simulationSpace=RootRelative
    alphaBoolBlend {blendOption=Linear}
    boundBone {boneName=dyn_cloak_02}
    componentAppliedLinearAccClamp=<2000 2000 1600>
    componentLinearAccScale=<0.4 0.4 0.4>
    componentLinearVelScale=<0.45 0.45 0.55>
    gravityOverride {z=-980}
    physicsBodyDefinitions [{
      sphereCollisionRadius=2
      boundBone {boneName=dyn_cloak_02}
      boxExtents=<8 13 13>
      constraintSetup {angularLimitsMax=<0 0 20>, angularLimitsMin=<0 -45 -20>, angularTarget=<0 0 0>}
      localJointOffset=<2 0 -0.1>
     }]
    relativeSpaceBone {boneName=root}
   }
   n32 = RigidBody!(ComponentPose=@n31, Alpha, OverrideWorldGravity:Vector=Get GravityOverride(), ExternalForce) -> (Pose) {
    bClampLinearTranslationLimitToRefPose=true
    bOverrideWorldGravity=true
    overridePhysicsAsset=&/F_MED_ToadCycle_Arm_Top_Physics
    simulationSpace=BaseBoneSpace
    alphaBoolBlend {blendOption=Linear}
    baseBoneRef {boneName=spine_05}
    componentAppliedLinearAccClamp=<0 0 0>
    overrideWorldGravity {z=-980}
    simSpaceSettings {
     dampingAlpha=0.25
     maxAngularAcceleration=28
     maxAngularVelocity=7
     maxLinearAcceleration=900
     maxLinearVelocity=600
     velocityScaleZ=0.4
     worldAlpha=1
     externalLinearDragV=<0.3 0.3 0.3>
    }
   }
   n33 = "Constraint - dyn_scarf_bk_01 (PP)"!(ComponentPose=@n32, bAlphaBoolEnabled:bool=True, ConstraintWeights_0=0, ConstraintWeights_1:float=1) -> (Pose) {
    alphaInputType=Bool
    boneToModify {boneName=dyn_scarf_bk_01}
    constraintSetup [{targetBone {boneName=spine_04}}, {targetBone {boneName=spine_05}}]
    constraintWeights [1,1]
   }
   n34 = "Constraint - dyn_scarf_bk_02 (PP)"!(ComponentPose=@n33, bAlphaBoolEnabled:bool=True, ConstraintWeights_0=0, ConstraintWeights_1:float=1) -> (Pose) {
    alphaInputType=Bool
    boneToModify {boneName=dyn_scarf_bk_02}
    constraintSetup [{targetBone {boneName=spine_04}}, {targetBone {boneName=spine_05}}]
    constraintWeights [1,1]
   }
   n35 = "Constraint - dyn_scarf_bk_03 (PP)"!(ComponentPose=@n34, bAlphaBoolEnabled:bool=True, ConstraintWeights_0=0, ConstraintWeights_1:float=1) -> (Pose) {
    alphaInputType=Bool
    boneToModify {boneName=dyn_scarf_bk_03}
    constraintSetup [{targetBone {boneName=spine_04}}, {targetBone {boneName=spine_05}}]
    constraintWeights [1,1]
   }
   n36 = "Constraint - dyn_scarf_bk_04 (PP)"!(ComponentPose=@n35, bAlphaBoolEnabled:bool=True, ConstraintWeights_0=0, ConstraintWeights_1:float=1) -> (Pose) {
    alphaInputType=Bool
    boneToModify {boneName=dyn_scarf_bk_04}
    constraintSetup [{targetBone {boneName=spine_04}}, {targetBone {boneName=spine_05}}]
    constraintWeights [1,1]
   }
   n37 = "Constraint - dyn_scarf_bk_05 (PP)"!(ComponentPose=@n36, bAlphaBoolEnabled:bool=True, ConstraintWeights_0=0.7, ConstraintWeights_1:float=0.3) -> (Pose) {
    alphaInputType=Bool
    boneToModify {boneName=dyn_scarf_bk_05}
    constraintSetup [{targetBone {boneName=spine_04}}, {targetBone {boneName=spine_05}}]
    constraintWeights [1,1]
   }
   n38 = "Constraint - dyn_scarf_bk_06 (PP)"!(ComponentPose=@n37, bAlphaBoolEnabled:bool=True, ConstraintWeights_0=0.5, ConstraintWeights_1:float=0.5) -> (Pose) {
    alphaInputType=Bool
    boneToModify {boneName=dyn_scarf_bk_06}
    constraintSetup [{targetBone {boneName=spine_04}}, {targetBone {boneName=spine_05}}]
    constraintWeights [1,1]
   }
   n39 = "Constraint - dyn_cloak_01 (PP)"!(ComponentPose=@n38, bAlphaBoolEnabled:bool=True, ConstraintWeights_0=0.7, ConstraintWeights_1:float=0.3) -> (Pose) {
    alphaInputType=Bool
    boneToModify {boneName=dyn_cloak_01}
    constraintSetup [{targetBone {boneName=spine_04}}, {targetBone {boneName=spine_05}}]
    constraintWeights [1,1]
   }
   n40 = "Constraint - dyn_cloak_02 (PP)"!(ComponentPose=@n39, bAlphaBoolEnabled:bool=True, ConstraintWeights_0=0.5, ConstraintWeights_1:float=0.5) -> (Pose) {
    alphaInputType=Bool
    boneToModify {boneName=dyn_cloak_02}
    constraintSetup [{targetBone {boneName=spine_04}}, {targetBone {boneName=spine_05}}]
    constraintWeights [1,1]
   }
   n41 = "Transform (Modify) Bone - Bone: dyn_Cloth_volume_l_B"!(ComponentPose=@n40, bAlphaBoolEnabled:bool=True, Translation="0,0,0", Rotation="0,-2,0", Scale="1,1,1") -> (Pose) {
    alphaInputType=Bool
    rotationMode=BMM_Additive
    rotationSpace=BCS_BoneSpace
    translationMode=BMM_Additive
    translationSpace=BCS_BoneSpace
    alphaBoolBlend {blendOption=Linear}
    boneToModify {boneName=dyn_Cloth_volume_l_B}
    rotation {yaw=-2}
   }
   n42 = "Transform (Modify) Bone - Bone: dyn_Cloth_volume_r_B"!(ComponentPose=@n41, bAlphaBoolEnabled:bool=True, Translation="0,0,0", Rotation="0,-7,0", Scale="1,1,1") -> (Pose) {
    alphaInputType=Bool
    rotationMode=BMM_Additive
    rotationSpace=BCS_BoneSpace
    translationMode=BMM_Additive
    translationSpace=BCS_BoneSpace
    alphaBoolBlend {blendOption=Linear}
    boneToModify {boneName=dyn_Cloth_volume_r_B}
    rotation {yaw=-7}
   }
   n43 = "Transform (Modify) Bone - Bone: dyn_Cloth_volume_l_A"!(ComponentPose=@n42, bAlphaBoolEnabled:bool=True, Translation="0,0,0", Rotation="0,-7,0", Scale="1,1,1") -> (Pose) {
    alphaInputType=Bool
    rotationMode=BMM_Additive
    rotationSpace=BCS_BoneSpace
    translationMode=BMM_Additive
    translationSpace=BCS_BoneSpace
    alphaBoolBlend {blendOption=Linear}
    boneToModify {boneName=dyn_Cloth_volume_l_A}
    rotation {yaw=-7}
   }
   n44 = "Transform (Modify) Bone - Bone: dyn_Cloth_volume_r_A"!(ComponentPose=@n43, bAlphaBoolEnabled:bool=True, Translation="0,0,0", Rotation="0,-7,0", Scale="1,1,1") -> (Pose) {
    alphaInputType=Bool
    rotationMode=BMM_Additive
    rotationSpace=BCS_BoneSpace
    translationMode=BMM_Additive
    translationSpace=BCS_BoneSpace
    alphaBoolBlend {blendOption=Linear}
    boneToModify {boneName=dyn_Cloth_volume_r_A}
    rotation {yaw=-7}
   }
   Save cached pose 'for diving'(Pose=Component To Local(ComponentPose=@n44)) {cacheName=for diving}
   n45 = Trail controller - Bone: dyn_cape_f2_l_05!(
    ComponentPose=Local To Component(LocalPose=Use cached pose 'ctrlRig_cache'())
    Alpha=1
    RelaxationSpeedScale:float=2
    FakeVelocity:Vector="0,0,0"
   ) -> (Pose) {
    bShowTrailLocation=true
    bUsePlanarLimit=true
    chainLength=4
    debugLifeTime=0.1
    lastBoneRotationAnimAlphaBlend=1
    maxDeltaTime=0.0333
    relaxationSpeedScale=1.2
    alphaScaleBiasClamp {bInterpResult=true, interpSpeedDecreasing=25, interpSpeedIncreasing=25}
    rotationLimits [{limitMax=<180 180 180>}, {limitMax=<180 180 180>}, {limitMax=<180 180 180>}, {limitMax=<180 180 180>}]
    rotationOffsets [<0 0 0>, <0 0 0>, <0 0 0>, <0 0 0>]
    trailBone {boneName=dyn_cape_f2_l_05}
    trailRelaxationSpeed {editorCurveData {keys [<at 0 15 RCIM_Linear>,<at 1 6 RCIM_Linear>]}}
   }
   n46 = Trail controller - Bone: dyn_cape_f2_r_05!(ComponentPose=@n45, Alpha=1, RelaxationSpeedScale:float=2, FakeVelocity:Vector="0,0,0") -> (Pose) {
    bShowTrailLocation=true
    bUsePlanarLimit=true
    chainLength=4
    debugLifeTime=0.1
    lastBoneRotationAnimAlphaBlend=1
    maxDeltaTime=0.0333
    relaxationSpeedScale=1.2
    alphaScaleBiasClamp {bInterpResult=true, interpSpeedDecreasing=25, interpSpeedIncreasing=25}
    rotationLimits [{limitMax=<180 180 180>}, {limitMax=<180 180 180>}, {limitMax=<180 180 180>}, {limitMax=<180 180 180>}]
    rotationOffsets [<0 0 0>, <0 0 0>, <0 0 0>, <0 0 0>]
    trailBone {boneName=dyn_cape_f2_r_05}
    trailRelaxationSpeed {editorCurveData {keys [<at 0 15 RCIM_Linear>,<at 1 6 RCIM_Linear>]}}
   }
   n47 = Trail controller - Bone: dyn_cape_f_l_05!(ComponentPose=@n46, Alpha=1, RelaxationSpeedScale:float=2, FakeVelocity:Vector="0,0,0") -> (Pose) {
    bShowTrailLocation=true
    bUsePlanarLimit=true
    chainLength=4
    debugLifeTime=0.1
    lastBoneRotationAnimAlphaBlend=1
    maxDeltaTime=0.0333
    relaxationSpeedScale=1.2
    alphaScaleBiasClamp {bInterpResult=true, interpSpeedDecreasing=25, interpSpeedIncreasing=25}
    rotationLimits [{limitMax=<180 180 180>}, {limitMax=<180 180 180>}, {limitMax=<180 180 180>}, {limitMax=<180 180 180>}]
    rotationOffsets [<0 0 0>, <0 0 0>, <0 0 0>, <0 0 0>]
    trailBone {boneName=dyn_cape_f_l_05}
    trailRelaxationSpeed {editorCurveData {keys [<at 0 15 RCIM_Linear>,<at 1 6 RCIM_Linear>]}}
   }
   n48 = Trail controller - Bone: dyn_cape_f_r_05!(ComponentPose=@n47, Alpha=1, RelaxationSpeedScale:float=2, FakeVelocity:Vector="0,0,0") -> (Pose) {
    bShowTrailLocation=true
    bUsePlanarLimit=true
    chainLength=4
    debugLifeTime=0.1
    lastBoneRotationAnimAlphaBlend=1
    maxDeltaTime=0.0333
    relaxationSpeedScale=1.2
    alphaScaleBiasClamp {bInterpResult=true, interpSpeedDecreasing=25, interpSpeedIncreasing=25}
    rotationLimits [{limitMax=<180 180 180>}, {limitMax=<180 180 180>}, {limitMax=<180 180 180>}, {limitMax=<180 180 180>}]
    rotationOffsets [<0 0 0>, <0 0 0>, <0 0 0>, <0 0 0>]
    trailBone {boneName=dyn_cape_f_r_05}
    trailRelaxationSpeed {editorCurveData {keys [<at 0 15 RCIM_Linear>,<at 1 6 RCIM_Linear>]}}
   }
   n49 = Trail controller - Bone: dyn_cape_s_l_05!(ComponentPose=@n48, Alpha=1, RelaxationSpeedScale:float=2, FakeVelocity:Vector="0,0,0") -> (Pose) {
    bReorientParentToChild=false
    bShowTrailLocation=true
    bUsePlanarLimit=true
    chainLength=4
    debugLifeTime=0.1
    lastBoneRotationAnimAlphaBlend=1
    maxDeltaTime=0.0333
    relaxationSpeedScale=1.2
    alphaScaleBiasClamp {bInterpResult=true, interpSpeedDecreasing=25, interpSpeedIncreasing=25}
    rotationLimits [{limitMax=<180 180 180>}, {limitMax=<180 180 180>}, {limitMax=<180 180 180>}, {limitMax=<180 180 180>}]
    rotationOffsets [<0 0 0>, <0 0 0>, <0 0 0>, <0 0 0>]
    trailBone {boneName=dyn_cape_s_l_05}
    trailRelaxationSpeed {editorCurveData {keys [<at 0 15 RCIM_Linear>,<at 1 6 RCIM_Linear>]}}
   }
   n50 = Trail controller - Bone: dyn_cape_s_r_05!(ComponentPose=@n49, Alpha=1, RelaxationSpeedScale:float=2, FakeVelocity:Vector="0,0,0") -> (Pose) {
    bShowTrailLocation=true
    bUsePlanarLimit=true
    chainLength=4
    debugLifeTime=0.1
    lastBoneRotationAnimAlphaBlend=1
    maxDeltaTime=0.0333
    relaxationSpeedScale=1.2
    alphaScaleBiasClamp {bInterpResult=true, interpSpeedDecreasing=25, interpSpeedIncreasing=25}
    rotationLimits [{limitMax=<180 180 180>}, {limitMax=<180 180 180>}, {limitMax=<180 180 180>}, {limitMax=<180 180 180>}]
    rotationOffsets [<0 0 0>, <0 0 0>, <0 0 0>, <0 0 0>]
    trailBone {boneName=dyn_cape_s_r_05}
    trailRelaxationSpeed {editorCurveData {keys [<at 0 15 RCIM_Linear>,<at 1 6 RCIM_Linear>]}}
   }
   n51 = Trail controller - Bone: dyn_cape_b_l_05!(ComponentPose=@n50, Alpha=1, RelaxationSpeedScale:float=2, FakeVelocity:Vector="0,0,0") -> (Pose) {
    bShowTrailLocation=true
    bUsePlanarLimit=true
    chainLength=3
    debugLifeTime=0.1
    lastBoneRotationAnimAlphaBlend=1
    maxDeltaTime=0.0333
    relaxationSpeedScale=1.2
    alphaScaleBiasClamp {bInterpResult=true, interpSpeedDecreasing=25, interpSpeedIncreasing=25}
    rotationLimits [{limitMax=<180 180 180>}, {limitMax=<180 180 180>}, {limitMax=<180 180 180>}]
    rotationOffsets [<0 0 0>, <0 0 0>, <0 0 0>]
    trailBone {boneName=dyn_cape_b_l_05}
    trailRelaxationSpeed {editorCurveData {keys [<at 0 15 RCIM_Linear>,<at 1 6 RCIM_Linear>]}}
   }
   n52 = Trail controller - Bone: dyn_cape_b_r_05!(ComponentPose=@n51, Alpha=1, RelaxationSpeedScale:float=2, FakeVelocity:Vector="0,0,0") -> (Pose) {
    bShowTrailLocation=true
    bUsePlanarLimit=true
    chainLength=3
    debugLifeTime=0.1
    lastBoneRotationAnimAlphaBlend=1
    maxDeltaTime=0.0333
    relaxationSpeedScale=1.2
    alphaScaleBiasClamp {bInterpResult=true, interpSpeedDecreasing=25, interpSpeedIncreasing=25}
    rotationLimits [{limitMax=<180 180 180>}, {limitMax=<180 180 180>}, {limitMax=<180 180 180>}]
    rotationOffsets [<0 0 0>, <0 0 0>, <0 0 0>]
    trailBone {boneName=dyn_cape_b_r_05}
    trailRelaxationSpeed {editorCurveData {keys [<at 0 15 RCIM_Linear>,<at 1 6 RCIM_Linear>]}}
   }
   Save cached pose 'dbnoCache'(Pose=Component To Local(ComponentPose=@n52)) {cacheName=dbnoCache}
   n53 = Control Rig!(
    Source=Copy Pose From Mesh() {bUseAttachedParent=true}
    Alpha
    isClothEnabled:bool=Get bClothEnabled()
    isSliding:bool=Get bIsSliding()
    isMoving:bool=Get bIsMoving2D()
    isCrouching:bool=Get bIsCrouching()
    isBaseballSliding:bool=Get bIsBaseballSliding()
   ) -> (Pose) {controlRigAssetReference {blueprintRigClass=&/F_MED_ToadCycle_Arm_CtrlRig.F_MED_ToadCycle_Arm_CtrlRig_C}}
   n54 = Anim Dynamics - Bone: dyn_Cloth_volume_l_A01!(ComponentPose=Local To Component(LocalPose=@n53), Alpha=1, GravityOverride:Vector="0,0,500", ExternalForce="0,0,0") -> (Pose) {
    angularDampingOverride=0.99
    angularSpringConstant=800
    bAngularSpring=true
    bOverrideAngularDamping=true
    bUseGravityOverride=true
    gravityScale=0.3
    simulationSpace=BoneRelative
    alphaBoolBlend {blendOption=Linear}
    boundBone {boneName=dyn_Cloth_volume_l_A01}
    componentAppliedLinearAccClamp=<2500 2500 1500>
    componentLinearAccScale=<0.5 0.5 0.4>
    componentLinearVelScale=<0.55 0.55 0.45>
    gravityOverride {z=500}
    physicsBodyDefinitions [{
      collisionType=CustomSphere
      sphereCollisionRadius=2
      boundBone {boneName=dyn_Cloth_volume_l_A01}
      boxExtents=<40 40 40>
      constraintSetup {angularLimitsMax=<0 0 20>, angularLimitsMin=<0 0 0>, angularTarget=<0 0 0>}
      localJointOffset=<-5 0.2 -0.1>
     }]
    relativeSpaceBone {boneName=root}
   }
   n55 = Anim Dynamics - Bone: dyn_Cloth_volume_l_A02!(ComponentPose=@n54, Alpha=1, GravityOverride:Vector="0,0,500", ExternalForce="0,0,0") -> (Pose) {
    angularDampingOverride=0.99
    angularSpringConstant=800
    bAngularSpring=true
    bOverrideAngularDamping=true
    bUseGravityOverride=true
    gravityScale=0.3
    simulationSpace=BoneRelative
    alphaBoolBlend {blendOption=Linear}
    boundBone {boneName=dyn_Cloth_volume_l_A02}
    componentAppliedLinearAccClamp=<2500 2500 1500>
    componentLinearAccScale=<0.5 0.5 0.4>
    componentLinearVelScale=<0.55 0.55 0.45>
    gravityOverride {z=500}
    physicsBodyDefinitions [{
      collisionType=CustomSphere
      sphereCollisionRadius=2
      boundBone {boneName=dyn_Cloth_volume_l_A02}
      boxExtents=<20 20 20>
      constraintSetup {angularLimitsMax=<0 0 30>, angularLimitsMin=<0 0 0>, angularTarget=<0 0 0>}
      localJointOffset=<-5 0.2 -0.1>
     }]
    relativeSpaceBone {boneName=root}
   }
   n56 = Anim Dynamics - Bone: dyn_Cloth_volume_r_A01!(ComponentPose=@n55, Alpha=1, GravityOverride:Vector="0,0,-980", ExternalForce="0,0,0") -> (Pose) {
    angularDampingOverride=0.99
    angularSpringConstant=800
    bAngularSpring=true
    bOverrideAngularDamping=true
    bUseGravityOverride=true
    gravityScale=0.3
    simulationSpace=BoneRelative
    alphaBoolBlend {blendOption=Linear}
    boundBone {boneName=dyn_Cloth_volume_r_A01}
    componentAppliedLinearAccClamp=<2500 2500 1500>
    componentLinearAccScale=<0.5 0.5 0.4>
    componentLinearVelScale=<0.55 0.55 0.45>
    gravityOverride {z=-980}
    physicsBodyDefinitions [{
      collisionType=CustomSphere
      sphereCollisionRadius=2
      boundBone {boneName=dyn_Cloth_volume_r_A01}
      boxExtents=<40 40 40>
      constraintSetup {angularLimitsMax=<1 1 20>, angularLimitsMin=<-1 -1 0>, angularTarget=<0 0 0>}
      localJointOffset=<-5 0.2 -0.1>
     }]
    relativeSpaceBone {boneName=root}
   }
   n57 = Anim Dynamics - Bone: dyn_Cloth_volume_r_A02!(ComponentPose=@n56, Alpha=1, GravityOverride:Vector="0,0,-980", ExternalForce="0,0,0") -> (Pose) {
    angularDampingOverride=0.99
    angularSpringConstant=800
    bAngularSpring=true
    bOverrideAngularDamping=true
    bUseGravityOverride=true
    gravityScale=0.3
    simulationSpace=BoneRelative
    alphaBoolBlend {blendOption=Linear}
    boundBone {boneName=dyn_Cloth_volume_r_A02}
    componentAppliedLinearAccClamp=<2500 2500 1500>
    componentLinearAccScale=<0.5 0.5 0.4>
    componentLinearVelScale=<0.55 0.55 0.45>
    gravityOverride {z=-980}
    physicsBodyDefinitions [{
      collisionType=CustomSphere
      sphereCollisionRadius=2
      boundBone {boneName=dyn_Cloth_volume_r_A02}
      boxExtents=<20 20 20>
      constraintSetup {angularLimitsMax=<1 1 30>, angularLimitsMin=<-1 -1 0>, angularTarget=<0 0 0>}
      localJointOffset=<-5 0.2 -0.1>
     }]
    relativeSpaceBone {boneName=root}
   }
   n58 = Anim Dynamics - Bone: dyn_Cloth_volume_fn_01!(ComponentPose=@n57, Alpha=1, GravityOverride:Vector="0,0,500", ExternalForce="0,0,0") -> (Pose) {
    angularDampingOverride=0.99
    angularSpringConstant=800
    bAngularSpring=true
    bOverrideAngularDamping=true
    bUseGravityOverride=true
    gravityScale=0.3
    simulationSpace=BoneRelative
    alphaBoolBlend {blendOption=Linear}
    boundBone {boneName=dyn_Cloth_volume_fn_01}
    componentAppliedLinearAccClamp=<2500 2500 1500>
    componentLinearAccScale=<0.5 0.5 0.4>
    componentLinearVelScale=<0.55 0.55 0.45>
    gravityOverride {z=500}
    physicsBodyDefinitions [{
      collisionType=CustomSphere
      sphereCollisionRadius=2
      boundBone {boneName=dyn_Cloth_volume_fn_01}
      boxExtents=<40 40 40>
      constraintSetup {angularLimitsMax=<0 3.5 20>, angularLimitsMin=<0 -3.5 0>, angularTarget=<0 0 0>}
      localJointOffset=<-5 0.2 -0.1>
     }]
    relativeSpaceBone {boneName=root}
   }
   n59 = Anim Dynamics - Bone: dyn_Cloth_volume_fn_02!(ComponentPose=@n58, Alpha=1, GravityOverride:Vector="0,0,500", ExternalForce="0,0,0") -> (Pose) {
    angularDampingOverride=0.99
    angularSpringConstant=800
    bAngularSpring=true
    bOverrideAngularDamping=true
    bUseGravityOverride=true
    gravityScale=0.3
    simulationSpace=BoneRelative
    alphaBoolBlend {blendOption=Linear}
    boundBone {boneName=dyn_Cloth_volume_fn_02}
    componentAppliedLinearAccClamp=<2500 2500 1500>
    componentLinearAccScale=<0.5 0.5 0.4>
    componentLinearVelScale=<0.55 0.55 0.45>
    gravityOverride {z=500}
    physicsBodyDefinitions [{
      collisionType=CustomSphere
      sphereCollisionRadius=2
      boundBone {boneName=dyn_Cloth_volume_fn_02}
      boxExtents=<20 20 20>
      constraintSetup {angularLimitsMax=<0 3.5 30>, angularLimitsMin=<0 -3.5 0>, angularTarget=<0 0 0>}
      localJointOffset=<-5 0.2 -0.1>
     }]
    relativeSpaceBone {boneName=root}
   }
   n60 = Anim Dynamics - Bone: dyn_cape_f2_l_base!(ComponentPose=@n59, Alpha=1, GravityOverride:Vector="0,0,500", ExternalForce="0,0,0") -> (Pose) {
    angularDampingOverride=1
    angularSpringConstant=800
    bAngularSpring=true
    bOverrideAngularDamping=true
    bUseGravityOverride=true
    gravityScale=0.3
    simulationSpace=BoneRelative
    alphaBoolBlend {blendOption=Linear}
    boundBone {boneName=dyn_cape_f2_l_base}
    componentAppliedLinearAccClamp=<2500 2500 1500>
    componentLinearAccScale=<0.5 0.5 0.4>
    componentLinearVelScale=<0.55 0.55 0.45>
    gravityOverride {z=500}
    physicsBodyDefinitions [{
      collisionType=CustomSphere
      sphereCollisionRadius=2
      boundBone {boneName=dyn_cape_f2_l_base}
      boxExtents=<40 40 40>
      constraintSetup {angularLimitsMax=<0 0 5>, angularLimitsMin=<0 -25 -5>, angularTarget=<0 0 0>}
      localJointOffset=<2.5 0 0>
     }]
    relativeSpaceBone {boneName=root}
   }
   n61 = Anim Dynamics - Bone: dyn_cape_f2_l_01!(ComponentPose=@n60, Alpha=1, GravityOverride:Vector="0,0,500", ExternalForce="0,0,0") -> (Pose) {
    angularDampingOverride=1
    angularSpringConstant=800
    bAngularSpring=true
    bOverrideAngularDamping=true
    bUseGravityOverride=true
    gravityScale=0.3
    simulationSpace=BoneRelative
    alphaBoolBlend {blendOption=Linear}
    boundBone {boneName=dyn_cape_f2_l_01}
    componentAppliedLinearAccClamp=<2500 2500 1500>
    componentLinearAccScale=<0.5 0.5 0.4>
    componentLinearVelScale=<0.55 0.55 0.45>
    gravityOverride {z=500}
    physicsBodyDefinitions [{
      collisionType=CustomSphere
      sphereCollisionRadius=2
      boundBone {boneName=dyn_cape_f2_l_01}
      boxExtents=<20 20 20>
      constraintSetup {angularLimitsMax=<0 0 5>, angularLimitsMin=<0 -25 -5>, angularTarget=<0 0 0>}
      localJointOffset=<2.5 0 0>
     }]
    relativeSpaceBone {boneName=root}
   }
   Save cached pose 'ctrlRig_cache'(Pose=Component To Local(ComponentPose=@n61)) {cacheName=ctrlRig_cache}
   n62 = RigidBody!(
    ComponentPose=Local To Component(LocalPose=Use cached pose 'ctrlRig_cache'())
    Alpha
    OverrideWorldGravity=Get GravityOverride()
    ExternalForce
    SimSpaceSettings:Sim Space Settings Structure=(WorldAlpha=1,VelocityScaleZ=0.4,DampingAlpha=0,MaxLinearVelocity=400,MaxAngularVelocity=8,MaxLinearAcceleration=1000,MaxAngularAcceleration=32,ExternalLinearDragV=(X=0.3,Y=0.3,Z=0.3),ExternalLinearVelocity=(X=0,Y=0,Z=0),ExternalAngularVelocity=(X=0,Y=0,Z=0))
   ) -> (Pose) {
    bClampLinearTranslationLimitToRefPose=true
    bOverrideWorldGravity=true
    overridePhysicsAsset=&/F_MED_ToadCycle_Arm_diving_Physics
    simulationSpace=BaseBoneSpace
    alphaBoolBlend {blendOption=Linear}
    baseBoneRef {boneName=pelvis}
    overrideWorldGravity {z=-980}
    simSpaceSettings {
     dampingAlpha=0
     maxAngularAcceleration=32
     maxAngularVelocity=8
     maxLinearAcceleration=1000
     maxLinearVelocity=400
     velocityScaleZ=0.4
     worldAlpha=1
     externalLinearDragV=<0.3 0.3 0.3>
    }
   }
   n63 = RigidBody(
    ComponentPose=Local To Component(LocalPose=Use cached pose 'ctrlRig_cache'())
    OverrideWorldGravity=Get GravityOverride()
    SimSpaceSettings=(WorldAlpha=1,VelocityScaleZ=0.4,DampingAlpha=0,MaxLinearVelocity=400,MaxAngularVelocity=8,MaxLinearAcceleration=1000,MaxAngularAcceleration=32,ExternalLinearDragV=(X=0.3,Y=0.3,Z=0.3),ExternalLinearVelocity=(X=0,Y=0,Z=0),ExternalAngularVelocity=(X=0,Y=0,Z=0))
   ) {
    bClampLinearTranslationLimitToRefPose=true
    bOverrideWorldGravity=true
    overridePhysicsAsset=&/F_MED_ToadCycle_Arm_Physics
    simulationSpace=BaseBoneSpace
    alphaBoolBlend {blendOption=Linear}
    baseBoneRef {boneName=pelvis}
    overrideWorldGravity {z=-980}
    simSpaceSettings {
     dampingAlpha=0
     maxAngularAcceleration=32
     maxAngularVelocity=8
     maxLinearAcceleration=1000
     maxLinearVelocity=400
     velocityScaleZ=0.4
     worldAlpha=1
     externalLinearDragV=<0.3 0.3 0.3>
    }
   }
   Save cached pose 'physicsCape_Cache'(Pose=Blend Poses by bool(BlendPose_0=Component To Local(ComponentPose=@n62), BlendPose_1=Component To Local(ComponentPose=@n63), bActiveValue=Get bIsSkyDiving()) {blendType=Linear}) {cacheName=physicsCape_Cache}
   n64 = Get BwdMovement() -> (BwdMovement:double)
   n65 = Trail controller - Bone: dyn_cape_f2_l_05!(ComponentPose=Local To Component(LocalPose=Use cached pose 'ctrlRig_cache'()), Alpha=@n64) -> (Pose) {
    bUsePlanarLimit=true
    chainLength=4
    debugLifeTime=0.1
    maxDeltaTime=0.333
    alphaScaleBiasClamp {bInterpResult=true}
    rotationLimits [{limitMax=<180 180 180>}, {limitMax=<180 180 180>}, {limitMax=<180 180 180>}, {limitMax=<180 180 180>}]
    rotationOffsets [<0 0 0>, <0 0 0>, <0 0 0>, <0 0 0>]
    trailBone {boneName=dyn_cape_f2_l_05}
    trailRelaxationSpeed {editorCurveData {keys [<at -0.0134 19.3841 RCIM_Linear>,<at 0.5079 12.8986 RCIM_Linear>,<at 0.9955 9.6044 RCIM_Linear>]}}
   }
   n66 = Trail controller - Bone: dyn_cape_f2_r_05!(ComponentPose=@n65, Alpha=@n64) -> (Pose) {
    bUsePlanarLimit=true
    chainLength=4
    debugLifeTime=0.1
    maxDeltaTime=0.333
    alphaScaleBiasClamp {bInterpResult=true}
    rotationLimits [{limitMax=<180 180 180>}, {limitMax=<180 180 180>}, {limitMax=<180 180 180>}, {limitMax=<180 180 180>}]
    rotationOffsets [<0 0 0>, <0 0 0>, <0 0 0>, <0 0 0>]
    trailBone {boneName=dyn_cape_f2_r_05}
    trailRelaxationSpeed {editorCurveData {keys [<at -0.0134 19.3841 RCIM_Linear>,<at 0.5079 12.8986 RCIM_Linear>,<at 0.9955 9.6044 RCIM_Linear>]}}
   }
   n67 = Trail controller - Bone: dyn_cape_f_l_05!(ComponentPose=@n66, Alpha=@n64) -> (Pose) {
    bUsePlanarLimit=true
    chainLength=4
    debugLifeTime=0.1
    maxDeltaTime=0.333
    alphaScaleBiasClamp {bInterpResult=true}
    rotationLimits [{limitMax=<180 180 180>}, {limitMax=<180 180 180>}, {limitMax=<180 180 180>}, {limitMax=<180 180 180>}]
    rotationOffsets [<0 0 0>, <0 0 0>, <0 0 0>, <0 0 0>]
    trailBone {boneName=dyn_cape_f_l_05}
    trailRelaxationSpeed {editorCurveData {keys [<at -0.0134 19.3841 RCIM_Linear>,<at 0.5079 12.8986 RCIM_Linear>,<at 0.9955 9.6044 RCIM_Linear>]}}
   }
   n68 = Trail controller - Bone: dyn_cape_f_r_05!(ComponentPose=@n67, Alpha=@n64) -> (Pose) {
    bUsePlanarLimit=true
    chainLength=4
    debugLifeTime=0.1
    maxDeltaTime=0.333
    alphaScaleBiasClamp {bInterpResult=true}
    rotationLimits [{limitMax=<180 180 180>}, {limitMax=<180 180 180>}, {limitMax=<180 180 180>}, {limitMax=<180 180 180>}]
    rotationOffsets [<0 0 0>, <0 0 0>, <0 0 0>, <0 0 0>]
    trailBone {boneName=dyn_cape_f_r_05}
    trailRelaxationSpeed {editorCurveData {keys [<at -0.0134 19.3841 RCIM_Linear>,<at 0.5079 12.8986 RCIM_Linear>,<at 0.9955 9.6044 RCIM_Linear>]}}
   }
   n69 = Trail controller - Bone: dyn_cape_s_l_05!(ComponentPose=@n68, Alpha=@n64) -> (Pose) {
    bUsePlanarLimit=true
    chainLength=4
    debugLifeTime=0.1
    maxDeltaTime=0.0333
    alphaScaleBiasClamp {bInterpResult=true}
    rotationLimits [{limitMax=<180 180 180>}, {limitMax=<180 180 180>}, {limitMax=<180 180 180>}, {limitMax=<180 180 180>}]
    rotationOffsets [<0 0 0>, <0 0 0>, <0 0 0>, <0 0 0>]
    trailBone {boneName=dyn_cape_s_l_05}
    trailRelaxationSpeed {editorCurveData {keys [<at 0 15 RCIM_Linear>,<at 0.5289 13.1838 RCIM_Linear>,<at 1 6 RCIM_Linear>]}}
   }
   n70 = Trail controller - Bone: dyn_cape_s_r_05!(ComponentPose=@n69, Alpha=@n64) -> (Pose) {
    bUsePlanarLimit=true
    chainLength=4
    debugLifeTime=0.1
    maxDeltaTime=0.333
    alphaScaleBiasClamp {bInterpResult=true}
    rotationLimits [{limitMax=<180 180 180>}, {limitMax=<180 180 180>}, {limitMax=<180 180 180>}, {limitMax=<180 180 180>}]
    rotationOffsets [<0 0 0>, <0 0 0>, <0 0 0>, <0 0 0>]
    trailBone {boneName=dyn_cape_s_r_05}
    trailRelaxationSpeed {editorCurveData {keys [<at -0.0134 19.3841 RCIM_Linear>,<at 0.5079 12.8986 RCIM_Linear>,<at 0.9955 9.6044 RCIM_Linear>]}}
   }
   n71 = Trail controller - Bone: dyn_cape_b_l_05!(ComponentPose=@n70, Alpha=@n64) -> (Pose) {
    bUsePlanarLimit=true
    chainLength=4
    debugLifeTime=0.1
    maxDeltaTime=0.0333
    alphaScaleBiasClamp {bInterpResult=true}
    rotationLimits [{limitMax=<180 180 180>}, {limitMax=<180 180 180>}, {limitMax=<180 90 180>, limitMin=<-180 -90 -180>}, {limitMax=<180 180 180>}]
    rotationOffsets [<0 0 0>, <0 0 0>, <0 0 0>, <0 0 0>]
    trailBone {boneName=dyn_cape_b_l_05}
    trailRelaxationSpeed {editorCurveData {keys [<at 0 15 RCIM_Linear>,<at 0.5289 13.1838 RCIM_Linear>,<at 1 6 RCIM_Linear>]}}
   }
   n72 = Trail controller - Bone: dyn_cape_b_r_05!(ComponentPose=@n71, Alpha=@n64) -> (Pose) {
    bUsePlanarLimit=true
    chainLength=4
    debugLifeTime=0.1
    maxDeltaTime=0.0333
    alphaScaleBiasClamp {bInterpResult=true}
    rotationLimits [{limitMax=<180 180 180>}, {limitMax=<180 180 180>}, {limitMax=<180 90 180>, limitMin=<-180 -90 -180>}, {limitMax=<180 180 180>}]
    rotationOffsets [<0 0 0>, <0 0 0>, <0 0 0>, <0 0 0>]
    trailBone {boneName=dyn_cape_b_r_05}
    trailRelaxationSpeed {editorCurveData {keys [<at 0 15 RCIM_Linear>,<at 0.5289 13.1838 RCIM_Linear>,<at 1 6 RCIM_Linear>]}}
   }
   Save cached pose 'trailCache'(Pose=Component To Local(ComponentPose=@n72)) {cacheName=trailCache}
  ]}
 EventGraph {nodes [
   n1 = Event BlueprintUpdateAnimation()
   > Set delta(delta:double=@n1.DeltaTimeX) -> (Output_Get:double)
   > n2 = Sequence!(execute) -> (then_0, then_1, then_2:Exec, then_3:Exec)
   > Branch(Condition=Get bRigidBodyEnabled())
   > n3 = "Gravity Override State Function (Parameter Asset) (Nativized)"(
    CharacterPartAnimInstance=Self-Reference()
    GameTimeInSeconds=GetGameTimeInSeconds()
    Parameters=&/F_MED_ToadCycle_Arm_Parameters
   )
   > Set GravityOverride(GravityOverride=@n3)
   > Set RBAN Settings Out(RBAN Settings Out=@n3.OutRigidBodyAnimNodeParameters)
   > Branch(Condition=Get bClothEnabled())
   > "Cloth Controller Function (Parameter Asset) (Nativized)"!(
    execute
    CharacterPartAnimInstance:Custom Character Part Anim Instance Object Reference=Self-Reference()
    ClothingAssetName:Name=F_MED_ToadCycle_Arm_Clothing_0
    ClothCoreSettings:"Cloth Core Settings Structure (by ref)"=Get Cloth Core Settings() -> (Cloth Core Settings:Cloth Core Settings Structure)
    WindGustLocalSettings:"Wind Gust Settings Structure (by ref)"=Get Wind Gust Local Settings() -> (Wind Gust Local Settings:Wind Gust Settings Structure)
    ClothUpAxis:Vector="0,0,1"
    LinVelScaleFwdBwdClamp:float=0
    LinVelScaleUpDwnClamp:float=0.7
    DeltaTime:float=Get delta() -> (delta:double)
    GameTimeInSeconds:float=GetGameTimeInSeconds()
    StateLogicLayerName:Name=none
    Parameters:Fort Character Dynamics Parameters Object Reference=&F_MED_ToadCycle_Arm_Parameters
   ) -> (
    then
    OutCurrentState:Name=None
    OutClothParameters:Cloth Parameters Structure
    OutClothCoreSettings:Cloth Core Settings Structure
    OutWindGustSettings:Wind Gust Parameters Structure
    OutWindGustLocalSettings:Wind Gust Settings Structure
    OutWindVelocity:Vector
    OutLinearVelVector:Vector
   )
   n4 = PlayerMovement(execute=@n2.then_1, AnimInstance=Self-Reference(), XDotProductCoefficient=1, YDotProductCoefficient=1, ZDotProductCoefficient=1)
   n5 = BreakVector(InVec=MakeVector(X=@n4, Y=@n4.YMovement, Z=@n4.ZMovement))
   n6 = "To Float (Boolean)"(InBool="float < float"(A=@n5, B=-50))
   > Set LeftMovement(LeftMovement:double=@n6) -> (Output_Get:double)
   n7 = "To Float (Boolean)"(InBool="float > float"(A=@n5, B=50))
   > Set RightMovement(RightMovement:double=@n7) -> (Output_Get:double)
   n8 = "To Float (Boolean)"(InBool="float > float"(A=@n5.Y, B=40))
   n9 = float + float!(A=@n6, B=@n7, C:double="To Float (Boolean)"(InBool="float < float"(A=@n5.Y, B=-40))) -> (ReturnValue)
   > Set FwdMovement(FwdMovement:double="Clamp (Float)"(Value=float - float(A=@n8, B=@n9), Min=0, Max=0.7)) -> (Output_Get:double)
   n10 = Get delta()
   n11 = Get FwdMovement() -> (FwdMovement:double)
   n12 = Get BwdMovement()
   > Set BwdMovement(BwdMovement:double=float * float(B=FInterpTo(Current=@n12, Target=@n11, DeltaTime=@n10, InterpSpeed=8)) {is=title K2Node_PromotableOperator}) -> (Output_Get:double)
   n13 = OR Boolean!(
    A=Get bIsPlayingFullBodySlotInFrontEnd() -> (bIsPlayingFullBodySlotInFrontEnd:bool)
    B=Get bIsPreviewingItemInFrontEnd() -> (bIsPreviewingItemInFrontEnd:bool)
    C:bool=Get bIsPlayingEmote()
    D:bool
   ) -> (ReturnValue)
   n14 = AND Boolean(A=Get bIsInFrontEnd(), B=Get bOverrideRBANSimSpaceInFrontEnd() -> (bOverrideRBANSimSpaceInFrontEnd:bool))
   n15 = Get bShouldDisplayFrontendReactivity() -> (bShouldDisplayFrontendReactivity:bool)
   Set is_Frontend And Not emote(execute=@n2.then_3, is_Frontend And Not emote:bool=AND Boolean(A=OR Boolean!(A=@n14, B=@n15) -> (ReturnValue), B=NOT Boolean(A=@n13))) -> (Output_Get:bool)
   n16 = Branch(execute=@n2.then_2, Condition=CharacterItemDefinitionHasAnyMetaTag(Tags=Make Array("[0]"=(TagName="Cosmetics.MetaTags.ToadCycleSarf"))))
   > Set On/Off_Scarf(On/Off_Scarf:double=1) -> (Output_Get:double)
   Set On/Off_Scarf(execute=@n16.else, On/Off_Scarf=0)
   Event OnCharacterCustomizationLoaded()
   n17 = CheckForBackbling()
   > Set isWearingPackBool(isWearingPackBool:bool=OR Boolean!(A=@n17, B=@n17.bIsBackblingACape, C:bool=@n17.bIsBackblingLong) -> (ReturnValue)) -> (Output_Get:bool)
   > Set is_WearingCape(is_WearingCape:bool=@n17.bIsBackblingLong) -> (Output_Get:bool)
   AND Boolean(A=@n17, B=NOT Boolean(A=@n17.bIsBackblingLong))
  ]}
]
variables [
 BwdMovement {type=number}
 ChaosClothingInteractor {type=ChaosClothingInteractor, default=null}
 Cloth Core Settings {
  type=ClothCoreSettings
  default {
   collisionThickness=1
   dampingCoefficient=0.1
   frictionCoefficient=0.3
   localDampingCoefficient=0.1
   numberIterations=2
   numberSubsteps=1
   selfCollisionThickness=0.65
   animDriveDamping=<0 1>
   animDriveStiffness=<0 1>
   areaStiffness=<1 1>
   bendingStiffness=<1 1>
   edgeStiffness=<1 1>
   tetherStiffness=<1 1>
  }
 }
 delta {type=number}
 FwdMovement {type=number}
 GravityOverride {type=Vector}
 is_Frontend And Not emote {type=boolean}
 is_WearingCape {type=boolean}
 isWearingPackBool {type=boolean}
 LeftMovement {type=number}
 On/Off_Scarf {type=number}
 RBAN Settings Out {type=FortRigidBodyAnimNodeParameters}
 RightMovement {type=number}
 Wind Gust Local Settings {type=WindGustSettings, default {elapsedTime=0, finalGustStrengthLocal=0, gustDuration=0, gustStrength=0, timeBetweenGusts=0}}
]
